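# Niagara skydiving contrail for the CanineCronut cosmetic: two mesh "planes" escort the diving character, with wing-tip trails and speed lines.
# Plane particles follow the skeletal mesh with yaw wobble, timeline-driven catch-up and orbit offsets, and smoothed banking.
# Trail emitters read plane position and orientation through particle attribute readers to place ribbons at the wings.
# `bIsFrontEndPreview` switches between gameplay and FE emitter copies; the visible smoke emitters are disabled.
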NiagaraSystem /BRCosmetics/Athena/Effects/Particles/Skydiving/NS_Contrail_CanineCronut/FX/NS_Contrail_CanineCronut
dependencies {
 usedDataInterfaces [
  &NiagaraDataInterfaceCurve
  &NiagaraDataInterfaceArrayFloat3
  &NiagaraDataInterfaceArrayFloat
  &NiagaraDataInterfaceParticleRead
  &NiagaraDataInterfaceColorCurve
  &NiagaraDataInterfaceVector2DCurve
  &NiagaraDataInterfaceSkeletalMesh
 ]
 usedDynamicInputs [
  &Vector2DFromCurve
  &InvertBool
  &Divide_Float
  &VectorLength
 ]
 usedModules [
  &self:SystemSpawnScript.SystemScriptSource.SystemScriptGraph.NiagaraNodeAssignment_0.SetVariables_#1
  &self:SystemSpawnScript.SystemScriptSource.SystemScriptGraph.NiagaraNodeAssignment_2.SetVariables_#2
  &self:SystemSpawnScript.SystemScriptSource.SystemScriptGraph.NiagaraNodeAssignment_6.SetVariables_#3
  &SystemState
  &Timeline
  /Game/Effects/Niagara/Modules/Skydiving/FN_DivingAttributes/v6/FN_DivingAttributes
  &self:ScratchModule
  &self:SystemSpawnScript.SystemScriptSource.SystemScriptGraph.NiagaraNodeAssignment_4.SetVariables_#4
  &self:SystemSpawnScript.SystemScriptSource.SystemScriptGraph.NiagaraNodeAssignment_1.SetVariables_#5
  &FN_Integrator
  &self:SystemSpawnScript.SystemScriptSource.SystemScriptGraph.NiagaraNodeAssignment_7.SetVariables_#6
  &self:SystemSpawnScript.SystemScriptSource.SystemScriptGraph.NiagaraNodeAssignment_9.SetVariables_#7
  &self:SystemSpawnScript.SystemScriptSource.SystemScriptGraph.NiagaraNodeAssignment_3.SetVariables_#8
  &self:SystemSpawnScript.SystemScriptSource.SystemScriptGraph.NiagaraNodeAssignment_5.SetVariables_#9
  &self:SystemSpawnScript.SystemScriptSource.SystemScriptGraph.NiagaraNodeAssignment_8.SetVariables_#10
 ]
}
emitters {
 FN_SpeedLines001 {
  emitterUpdateScript [
   &EmitterState(Life Cycle Mode=::NewEnumerator1, Loop Duration=1, UseLoopDelay=false)
   &SpawnRate(
    SpawnRate=&Lerp_Float(
     A=0
     B=&FN_FrontEndFloatModifier(
      Gameplay=&Multiply_Float(A=55, B=$User.ParaGlideLeanAlpha)
      Front End=&Multiply_Float(A=50, B=&Lerp_Float(A=0.5, B=1, Alpha=$User.ParaGlideLeanAlpha))
     )
     Alpha=1
    )
    Use Spawn Probability=true
   )
   &SpawnRate(
    Spawn Probability=&MakeCustomFloatFromBool(Bool=&SetBoolByFloatComparison(A=$System.Age, B=0.2), True Float=1, False Float=1)
    SpawnRate=&Lerp_Float(A=0, B=500, Alpha=$System.Dive)
    Use Spawn Probability=true
   ) {enabled=false, module=SpawnRate001}
  ]
  particleSpawnScript [
   &InitializeParticle(
    Color Maximum=<rgba 0.25 0.25 0.25 0.1>
    Color Minimum=<rgba 1 1 1 0.05>
    Color Mode=::NewEnumerator2
    Lifetime Max=0.5
    Lifetime Min=0.3
    Lifetime Mode=::NewEnumerator1
    Sprite Rotation Mode=::NewEnumerator2
    Sprite Size=&UniformRangedVector2D(Minimum=<7 100>, Maximum=<1 200>)
    Sprite Size Mode=::NewEnumerator4
    UsePositionOffset=false
    Write Material Random=true
   ) {module=InitializeParticle001}
   &SkeletalMeshLocation(
    Bone / Socket Index=3
    Bone Sampling Mode=::NewEnumerator4
    MaskSpawnGroup=false
    Sampled Orientation Coordinate Space=::Local
    Skeletal Mesh=$System.SkeletalMesh
   )
   &SphereLocation(
    MaskSpawnGroup=false
    Non Uniform Scale=&RotateVectorByQuaternion(Vector To Rotate=<1 1 1>, Quaternion=$Particles.SkeletalMeshLocation.SampledRotation)
    Offset=&FN_FrontEndVector3Modifier(
     Gameplay=&Add_Vector(
      A=&RotateVectorByQuaternion(Vector To Rotate=<100 0 0>, Quaternion=$Particles.SkeletalMeshLocation.SampledRotation)
      B=&TransformVector(Vector=<0 0 -200>, Source Space=::World, Destination Space=::World)
     )
     Front End=<0 0 0>
    )
    Override Seed=false
    Sphere Orientation Axis:Vector3f=&RotateVectorByQuaternion(Vector To Rotate=<100 0 0>, Quaternion=$Particles.SkeletalMeshLocation.SampledRotation)
    Sphere Origin:NiagaraPosition=$Particles.SkeletalMeshLocation.SampledPosition
    Sphere Radius=250
    Surface Only=false
    UseNonUniformScale=true
    UseOffset=true
    UseSphereOrientation=true
   )
   &InheritVelocity(Module.Limit Speed=true)
   &AddVelocity(
    Velocity=&FN_FrontEndVector3Modifier(Gameplay=<0 0 0>, Front End=<0 -100 100>)
    Velocity Speed Scale=&FN_FrontEndFloatModifier(Gameplay=1, Front End=2)
   )
  ]
  particleUpdateScript [
   &ParticleState()
   &SolveForcesAndVelocity(Clamp Velocity=false, Limit Acceleration=false)
   &ScaleSpriteSize(
    Non-Uniform Curve Scale=&FN_FrontEndVector2Modifier(Front End:Vector2f=<0.5 2>, Gameplay:Vector2f=<1.5 3>)
    Non-Uniform Curve Sprite Scale=DataInterface() {propertyValues {
      XCurve {keys [<at -0.0069 0.5 RCIM_Cubic>,<at 0.4918 1 RCIM_Cubic>,<at 1 0.5 RCIM_Cubic>]}
      YCurve {keys [
        {
         arriveTangent=3.1787
         arriveTangentWeight=0.684
         interpMode=RCIM_Cubic
         leaveTangent=3.1787
         leaveTangentWeight=0.684
         tangentMode=RCTM_User
         tangentWeightMode=RCTWM_WeightedBoth
         time=0
         value=0.25
        }
        {
         arriveTangentWeight=0.2354
         interpMode=RCIM_Cubic
         leaveTangentWeight=0.3527
         tangentMode=RCTM_Break
         tangentWeightMode=RCTWM_WeightedBoth
         time=0.4975
         value=1
        }
        <at 1 0.25 RCIM_Cubic>
       ]}
     }}
    Scale Sprite Size Mode=::NewEnumerator3
   )
   &ScaleColor(
    Scale Alpha=&FloatFromCurve(
     FloatCurve=DataInterface() {propertyValues {Curve {
        defaultValue=1
        keys [
         {arriveTangent=9.0725, interpMode=RCIM_Cubic, leaveTangent=9.0725, tangentMode=RCTM_User, time=0, value=0}
         {arriveTangent=-0.0194, interpMode=RCIM_Cubic, leaveTangent=-0.0194, tangentMode=RCTM_User, time=0.3, value=0.875}
         {arriveTangent=-0.0405, interpMode=RCIM_Cubic, leaveTangent=-0.0405, tangentMode=RCTM_User, time=1, value=0}
        ]
       }}}
     Scale Curve=&FN_FrontEndFloatModifier(Gameplay=0.7, Front End=0.35)
    )
    ScaleA=true
    ScaleRGB=true
   )
  ]
  rendererClasses [&NiagaraSpriteRendererProperties]
  renderers [{rendererClass=&NiagaraSpriteRendererProperties}]
 }
 LeftWingTrail {
  emitterUpdateScript [
   &EmitterState()
   &SpawnParticlesFromOtherEmitter(
    Attribute Reader=DataInterface() {propertyValues {EmitterBinding {bindingMode=Other, emitterName=Plane}}}
    Spawn Rate=&MakeCustomFloatFromBool(
     Bool=$User.bIsFrontEndPreview
     True Float=0
     False Float=&Lerp_Float(A=0, B=600, Alpha=&MakeCustomFloatFromBool(Bool=$System.SpeedBool, True Float=1, False Float=0))
    )
    Spawn Rate Per Particle Cap=1000
    Use Maximum Age=false
    Use Minimum Age=false
   )
  ]
  particleSpawnScript [
   &InitializeParticle(Color=<rgba 0.4271 0.4271 0.4271 0.5>, Lifetime=0.01, Sprite Size Mode=::NewEnumerator3, UsePositionOffset=true, Write Material Random=true)
   &SampleParticlesFromOtherEmitter(
    Mesh Orientation Sampling:ENiagara_DirectReadSamplingMode=::NewEnumerator1
    Particle ID Sampling=::NewEnumerator1
    Position Sampling:ENiagara_DirectReadSamplingMode=::NewEnumerator1
   )
   /SetVariables() {isSetParameters=true, module=SetVariables_#11}
   /SetVariables(Particles.RibbonID=$Output.SampleParticlesFromOtherEmitter.SampledID) {isSetParameters=true, module=SetVariables_#12}
   &ShapeLocation(
    Additional Quaternion Rotation=true
    Additional Yaw / Pitch / Roll=<0 90 0>
    Cylinder Distribution=::NewEnumerator2
    Cylinder Height=68
    Cylinder Radius=6
    MaskSpawnGroup=false
    Offset Coordinate Space=::Local
    Offset Mode=::NewEnumerator0
    Rotation Angle Type:ENiagara_AngleInput=::NewEnumerator0
    Rotation Mode=::NewEnumerator3
    Rotation Quaternion=$Output.SampleParticlesFromOtherEmitter.SampledMeshOrientation
    Shape Origin=$Output.SampleParticlesFromOtherEmitter.SampledPosition
    Shape Primitive=::NewEnumerator1
    U Position=&Lerp_Float(A=0.39, B=1, Alpha=0)
    UseOwnerScale=true
    V Position=1
   ) {module=ShapeLocation001}
  ]
  particleUpdateScript [
   &ParticleState()
   /Niagara/Modules/AttributeReader/UpdateParticlesFromOtherEmitter(
    Attribute Reader:NiagaraDataInterfaceParticleRead=$Emitter.SpawnParticlesAttributeReader
    Mesh Orientation Sampling:ENiagara_DirectReadSamplingMode=::NewEnumerator1
   ) {enabled=false}
   &ShapeLocation(
    Additional Quaternion Rotation=true
    Additional Yaw / Pitch / Roll=<0 90 0>
    Cylinder Distribution=::NewEnumerator2
    Cylinder Height=68
    Cylinder Radius=7
    MaskSpawnGroup=false
    Offset Coordinate Space=::World
    Offset Mode=::NewEnumerator0
    Rotation Angle Type:ENiagara_AngleInput=::NewEnumerator0
    Rotation Mode=::NewEnumerator3
    Rotation Quaternion=$Output.UpdateParticlesFromOtherEmitter.SampledMeshOrientation
    Shape Origin=$Output.UpdateParticlesFromOtherEmitter.SampledPosition
    Shape Primitive=::NewEnumerator1
    U Position=&Lerp_Float(A=0.4, B=1, Alpha=0)
    UseOwnerScale=true
    V Position=1
   ) {enabled=false}
   &SolveForcesAndVelocity(Clamp Velocity=false, Limit Acceleration=false)
   &ScaleRibbonWidth(Ribbon Width Scale=&FloatFromCurve(FloatCurve=DataInterface() {propertyValues {Curve {keys [<at 0 1>,<at 1 0>]}}}))
  ]
  rendererClasses [&NiagaraRibbonRendererProperties, &NiagaraSpriteRendererProperties]
  renderers [
   {rendererClass=&NiagaraRibbonRendererProperties}
   {rendererClass=&NiagaraSpriteRendererProperties, rendererIndex=1}
  ]
 }
 LeftWingTrailFE {
  emitterUpdateScript [
   &EmitterState()
   &SpawnParticlesFromOtherEmitter(
    Attribute Reader=DataInterface() {propertyValues {EmitterBinding {bindingMode=Other, emitterName=PlaneFE}}}
    Spawn Rate=&MakeCustomFloatFromBool(Bool=$User.bIsFrontEndPreview, True Float=&Lerp_Float(A=0, B=600, Alpha=$User.ParaGlideLeanAlpha), False Float=0)
    Spawn Rate Per Particle Cap=1000
    Use Maximum Age=false
    Use Minimum Age=false
   )
  ]
  particleSpawnScript [
   &InitializeParticle(Color=<rgba 0.4271 0.4271 0.4271 0.1>, Lifetime=0.01, Sprite Size Mode=::NewEnumerator3, UsePositionOffset=true, Write Material Random=true)
   &SampleParticlesFromOtherEmitter(
    Mesh Orientation Sampling:ENiagara_DirectReadSamplingMode=::NewEnumerator1
    Particle ID Sampling=::NewEnumerator1
    Position Sampling:ENiagara_DirectReadSamplingMode=::NewEnumerator1
   )
   /SetVariables() {isSetParameters=true, module=SetVariables_#11}
   /SetVariables(Particles.RibbonID=$Output.SampleParticlesFromOtherEmitter.SampledID) {isSetParameters=true, module=SetVariables_#12}
   &ShapeLocation(
    Additional Quaternion Rotation=true
    Additional Yaw / Pitch / Roll=<0 90 0>
    Cylinder Distribution=::NewEnumerator2
    Cylinder Height=68
    Cylinder Radius=6
    MaskSpawnGroup=false
    Offset Coordinate Space=::Local
    Offset Mode=::NewEnumerator0
    Rotation Angle Type:ENiagara_AngleInput=::NewEnumerator0
    Rotation Mode=::NewEnumerator3
    Rotation Quaternion=$Output.SampleParticlesFromOtherEmitter.SampledMeshOrientation
    Shape Origin=$Output.SampleParticlesFromOtherEmitter.SampledPosition
    Shape Primitive=::NewEnumerator1
    U Position=&Lerp_Float(A=0.126, B=1, Alpha=0)
    UseOwnerScale=true
    V Position=1
   ) {module=ShapeLocation001}
   &LinearForce(Force=<0 -1000 3000>)
   &ApplyInitialForces()
  ]
  particleUpdateScript [
   &ParticleState()
   &UpdateParticlesFromOtherEmitter(
    Attribute Reader:NiagaraDataInterfaceParticleRead=$Emitter.SpawnParticlesAttributeReader
    Mesh Orientation Sampling:ENiagara_DirectReadSamplingMode=::NewEnumerator1
   ) {enabled=false}
   &ShapeLocation(
    Additional Quaternion Rotation=true
    Additional Yaw / Pitch / Roll=<0 90 0>
    Cylinder Distribution=::NewEnumerator2
    Cylinder Height=68
    Cylinder Radius=7
    MaskSpawnGroup=false
    Offset Coordinate Space=::World
    Offset Mode=::NewEnumerator0
    Rotation Angle Type:ENiagara_AngleInput=::NewEnumerator0
    Rotation Mode=::NewEnumerator3
    Rotation Quaternion=$Output.UpdateParticlesFromOtherEmitter.SampledMeshOrientation
    Shape Origin=$Output.UpdateParticlesFromOtherEmitter.SampledPosition
    Shape Primitive=::NewEnumerator1
    U Position=&Lerp_Float(A=0.4, B=1, Alpha=0)
    UseOwnerScale=true
    V Position=1
   ) {enabled=false}
   &SolveForcesAndVelocity(Clamp Velocity=false, Limit Acceleration=false)
   &ScaleRibbonWidth(Ribbon Width Scale=&FloatFromCurve(FloatCurve=DataInterface() {propertyValues {Curve {keys [<at 0 1>,<at 1 0>]}}}))
  ]
  rendererClasses [&NiagaraRibbonRendererProperties, &NiagaraSpriteRendererProperties]
  renderers [
   {rendererClass=&NiagaraRibbonRendererProperties}
   {rendererClass=&NiagaraSpriteRendererProperties, rendererIndex=1}
  ]
 }
 Plane {
  emitterUpdateScript [
   &EmitterState()
   &SpawnBurst_Instantaneous(
    Spawn Count=&MakeCustomIntFromBool(Bool=$User.bIsFrontEndPreview, True INT=0, False INT=2)
    Use Spawn Probability=false
    UseLoopCountLimit=false
   )
   &InterpolateOverTime(Rate Of Change=50, Target Vector Value:Vector3f=$Engine.Owner.Velocity, Variable Type=::NewEnumerator2)
  ]
  particleSpawnScript [
   &InitializeParticle(Lifetime=5, Mesh Scale Mode=::NewEnumerator3, Sprite Size Mode=::NewEnumerator3, UsePositionOffset=true, Write Material Random=true)
   &ShapeLocation(MaskSpawnGroup=false, Override Seed=false, UseOwnerScale=true)
   /SetVariables(Particles.VisibilityTag=$Emitter.SpawnGroup) {isSetParameters=true, module=SetVariables_#13}
   /SetVariables(Particles.WingSize:Vector3f=&Multiply_VectorByFloat(Vector=<1 1.8748 1>, Float=1)) {isSetParameters=true, module=SetVariables_#14}
  ]
  particleUpdateScript [
   &ParticleState(Kill Particles When Lifetime Has Elapsed=false, Let Infinitely Lived Particles Die When Emitter Deactivates=true)
   /SetVariables(Particles.Ambient:Vector3f=&Multiply_VectorByFloat(
     Vector=&MakeVector(
      X=&Waveform(
       "Waveform [1]"=::NewEnumerator12
       "[1] Input Position"=&RandomRangeFloat(Minimum=0, Maximum=1000)
       "[1] Frequency"=1
       "[1] Amplitude Min/Max"=<-50 50>
      )
      Y=&Waveform(
       "Waveform [1]"=::NewEnumerator12
       "[1] Input Position"=&RandomRangeFloat(Minimum=1200, Maximum=1800)
       "[1] Frequency"=1
       "[1] Amplitude Min/Max"=<-50 50>
      )
      Z=&Waveform(
       "Waveform [1]"=::NewEnumerator12
       "[1] Input Position"=&RandomRangeFloat(Minimum=2200, Maximum=2800)
       "[1] Frequency"=1
       "[1] Amplitude Min/Max"=<-50 50>
      )
     )
     Float=0.35
    )) {isSetParameters=true, module=SetVariables_#15}
   &InterpolateOverTime(
    Rate Of Change=3
    Target Float Value:NiagaraFloat=&MakeCustomFloatFromBool(Bool=$System.Timeline001.IsRewinding, True Float=0, False Float=1)
   ) {module=InterpolateOverTime004}
   /SetVariables(Particles.CatchupVector:Vector3f=&Multiply_VectorByFloat(
     Vector=&Multiply_VectorByFloat(
      Vector=&NormalizeVector(Vector To Normalize=$System.SkydiveVelocity, Safe Normalize=true, Safe Fallback Threshold=0, Safe Fallback Vector=<0 0 0.001>)
      Float=-75
     )
     Float=&Lerp_Float(
      A=&FloatFromCurve(
       FloatCurve=DataInterface() {propertyValues {Curve {keys [
           {interpMode=RCIM_Cubic, tangentMode=RCTM_User, time=0, value=0}
           {interpMode=RCIM_Cubic, tangentMode=RCTM_User, time=0.35, value=-0.35}
           {tangentMode=RCTM_User, time=1, value=0}
          ]}}}
       CurveIndex=$System.Timeline001.NormalizedTime
      )
      B=&FloatFromCurve(
       FloatCurve=DataInterface() {propertyValues {Curve {keys [
           {arriveTangent=3.3827, arriveTangentWeight=1.1106, interpMode=RCIM_Cubic, leaveTangentWeight=1.1106, time=0, value=0}
           {arriveTangentWeight=0.5241, interpMode=RCIM_Cubic, leaveTangentWeight=0.1667, time=0.5, value=0.65}
           <at 1 0 RCIM_Cubic>
          ]}}}
       CurveIndex=$System.Timeline001.NormalizedTime
      )
      Alpha=$Particles.InterpolateOverTime004.InterpolateOverTime_Float.Moving Average
     )
    )) {isSetParameters=true, module=SetVariables_#16}
   &SkeletalMeshLocation(
    Bone Sampling Mode=::NewEnumerator4
    MaskSpawnGroup=false
    Orientation Sampling=::NewEnumerator1
    Sampled Position Offset=&Add_Vector(
     A=$Particles.CatchupVector
     B=/Niagara/DynamicInputs/Vectors/RotateVector(
      Vector To Rotate:Vector3f=&Add_Vector(
       A=<0 0 0>
       B=&Add_Vector(
        A=&SelectVectorFromArray(
         Direct Array Index=&ReturnExecIndex()
         Direct Array Mode=::NewEnumerator1
         Vector Selection Array=DataInterface() {propertyValues {FloatData [<75 0 100>, <-75 0 100>]}}
        )
        B=$System.OverallTargetOffset
       )
      )
      Yaw:NiagaraFloat=&Waveform(
       "Waveform [1]"=::NewEnumerator12
       "[1] Input Position"=&RandomRangeFloat(Minimum=200000, Maximum=100000)
       "[1] Frequency"=1.25
       "[1] Amplitude Min/Max"=<7 -7>
      )
      Pitch:NiagaraFloat=0
      Roll:NiagaraFloat=0
      Rotation Rate:NiagaraFloat=1
     )
    )
    Skeletal Mesh=$System.SkeletalMesh
    Use Sampled Mesh Attribute as Ribbon ID=false
   )
   &ScaleMeshSize(Scale Factor=&VectorFromFloat(Value=&FloatFromCurve(
      FloatCurve=DataInterface() {propertyValues {Curve {keys [{arriveTangent=1.0186, interpMode=RCIM_Cubic, leaveTangent=1.0186, tangentMode=RCTM_User, time=0, value=0}, <at 1 1 RCIM_Cubic>]}}}
      CurveIndex=$Particles.Age
     )))
   &Drag(Drag=2, Use Linear Drag=true, Use Rotational Drag=false)
   &SolveForcesAndVelocity(Clamp Velocity=false, Limit Acceleration=false)
   /SetVariables(Particles.MeshOrientation=$Engine.Owner.Rotation) {isSetParameters=true, module=SetVariables_#17}
   &UpdateMeshOrientation(
    Banking Angle Min / Max:Vector2f=<-1 1>
    "Foward Vector (Mesh Space)":Vector3f=<0 1 0>
    Orientation Method=::NewEnumerator3
    Particle Velocity:Vector3f=$Emitter.InterpolateOverTime.InterpolateOverTime_V3_.Moving Average
    "Right Vector (Mesh Space)":Vector3f=&Lerp_Vector(A=<0.5 0 0.5>, B=<0 1 0>, Alpha=$User.ParaGlideLeanAlpha)
    Up Vector:Vector3f=&Lerp_Vector(
     A=&Multiply_VectorByFloat(
      Vector=&ReturnMeshOrientationAxis(Mesh Orientation Axis=::YAxis, Mesh Orientation Quat=$Particles.MeshOrientation)
      Float=1
     )
     B=&Multiply_Vector(
      A=&ReturnMeshOrientationAxis(Mesh Orientation Axis=::ZAxis, Mesh Orientation Quat=$Particles.MeshOrientation)
      B=<1 1 1>
     )
     Alpha=&FloatFromCurve(FloatCurve=DataInterface() {propertyValues {Curve {keys [<at 0 0.25>,<at 0.25 0.75>]}}}, CurveIndex=$System.Dive)
    )
    Velocity Orientation Convergence Rate=100
   )
   &InterpolateOverTime(
    Rate Of Change:NiagaraFloat=&Lerp_Float(A=2, B=3, Alpha=$System.Dive)
    Target Quat Value:Quat4f=$Particles.MeshOrientation
    Variable Type=::NewEnumerator5
   ) {module=InterpolateOverTime001}
   /SetVariables(Particles.NewOrient:Quat4f=$Particles.InterpolateOverTime001.InterpolateOverTime_Quaternion.Moving Average) {isSetParameters=true, module=SetVariables_#18}
   /SetVariables(Particles.MeshOrientation=$Particles.NewOrient) {isSetParameters=true, module=SetVariables_#19}
   &RotateAroundPoint(
    Radius=110
    Rotation Center:NiagaraPosition=$Particles.SkeletalMeshLocation.SampledPosition
    Rotation Phase:NiagaraFloat=&Add_Float(
     A=$System.Timeline002.NormalizedTime
     B=&SelectFloatFromArray(
      Direct Array Index=&ReturnExecIndex()
      Float Selection Array=DataInterface() {propertyValues {FloatData [0,0.5]}}
     )
    )
    Rotation Rate:NiagaraFloat=&MakeCustomFloatFromBool(Bool=&SetBoolByFloatComparison(A=$System.Timeline002.NormalizedTime, B=0), True Float=0, False Float=1)
    Set Particle Position=false
   )
   /SetVariables(Particles.PrevPosition:NiagaraPosition=$Particles.Position) {isSetParameters=true, module=SetVariables_#20}
   /SetVariables(Particles.Position=&Lerp_Vector(
     A=Linked() {linkedVariable {name=Particles.PrevPosition, type=NiagaraPosition}}
     B=&Add_Vector(A=Linked() {linkedVariable {name=Output.RotateAroundPoint.OutputPosition, type=NiagaraPosition}}, B=<0 0 100>)
     Alpha=&FloatFromCurve(
      FloatCurve=DataInterface() {propertyValues {Curve {defaultValue=1, keys [<at 0 0 RCIM_Cubic>,<at 1 1 RCIM_Cubic>]}}}
      CurveIndex=$System.Timeline003.NormalizedTime
     )
    )) {isSetParameters=true, module=SetVariables_#21}
   /SetVariables(
    Particles.BladesColor:LinearColor=<rgba 0.4167 0.4167 0.4167 2>
    Particles.BladesDynamic:Vector4f=&MakeVector4(X=8, Y=0, Z=0, W=0)
   ) {isSetParameters=true, module=SetVariables_#22}
  ]
  rendererClasses [&NiagaraSpriteRendererProperties, &NiagaraMeshRendererProperties]
  renderers [
   {rendererClass=&NiagaraSpriteRendererProperties}
   {rendererClass=&NiagaraMeshRendererProperties, rendererIndex=1}
  ]
 }
 PlaneFE {
  emitterUpdateScript [
   &EmitterState()
   &SpawnBurst_Instantaneous(
    Spawn Count=&MakeCustomIntFromBool(Bool=$User.bIsFrontEndPreview, True INT=2, False INT=0)
    Use Spawn Probability=false
    UseLoopCountLimit=false
   )
   &InterpolateOverTime(Rate Of Change=50, Target Vector Value:Vector3f=$Engine.Owner.Velocity, Variable Type=::NewEnumerator2)
  ]
  particleSpawnScript [
   &InitializeParticle(Lifetime=5, Mesh Scale Mode=::NewEnumerator3, Sprite Size Mode=::NewEnumerator3, UsePositionOffset=true, Write Material Random=true)
   &ShapeLocation(MaskSpawnGroup=false, Override Seed=false, UseOwnerScale=true)
   /SetVariables(Particles.VisibilityTag=$Emitter.SpawnGroup) {isSetParameters=true, module=SetVariables_#13}
   /SetVariables(Particles.WingSize:Vector3f=&Multiply_VectorByFloat(Vector=<1 1.8748 1>, Float=1)) {isSetParameters=true, module=SetVariables_#14}
  ]
  particleUpdateScript [
   &ParticleState(Kill Particles When Lifetime Has Elapsed=false, Let Infinitely Lived Particles Die When Emitter Deactivates=true)
   /SetVariables(Particles.Ambient:Vector3f=&Multiply_VectorByFloat(
     Vector=&MakeVector(
      X=&Waveform(
       "Waveform [1]"=::NewEnumerator12
       "[1] Input Position"=&RandomRangeFloat(Minimum=0, Maximum=1000)
       "[1] Frequency"=1
       "[1] Amplitude Min/Max"=<-50 50>
      )
      Y=&Waveform(
       "Waveform [1]"=::NewEnumerator12
       "[1] Input Position"=&RandomRangeFloat(Minimum=1200, Maximum=1800)
       "[1] Frequency"=1
       "[1] Amplitude Min/Max"=<-50 50>
      )
      Z=&Waveform(
       "Waveform [1]"=::NewEnumerator12
       "[1] Input Position"=&RandomRangeFloat(Minimum=2200, Maximum=2800)
       "[1] Frequency"=1
       "[1] Amplitude Min/Max"=<-50 50>
      )
     )
     Float=0.35
    )) {isSetParameters=true, module=SetVariables_#15}
   &InterpolateOverTime(
    Rate Of Change=3
    Target Float Value:NiagaraFloat=&MakeCustomFloatFromBool(Bool=$System.Timeline001.IsRewinding, True Float=0, False Float=1)
   ) {module=InterpolateOverTime004}
   /SetVariables(Particles.CatchupVector:Vector3f=&Multiply_VectorByFloat(
     Vector=&Multiply_VectorByFloat(
      Vector=&NormalizeVector(Vector To Normalize=$System.SkydiveVelocity, Safe Normalize=true, Safe Fallback Threshold=0, Safe Fallback Vector=<0 0 0.001>)
      Float=-75
     )
     Float=&Lerp_Float(
      A=&FloatFromCurve(
       FloatCurve=DataInterface() {propertyValues {Curve {keys [
           {interpMode=RCIM_Cubic, tangentMode=RCTM_User, time=0, value=0}
           {interpMode=RCIM_Cubic, tangentMode=RCTM_User, time=0.35, value=-0.35}
           {tangentMode=RCTM_User, time=1, value=0}
          ]}}}
       CurveIndex=$System.Timeline001.NormalizedTime
      )
      B=&FloatFromCurve(
       FloatCurve=DataInterface() {propertyValues {Curve {keys [
           {arriveTangent=3.3827, arriveTangentWeight=1.1106, interpMode=RCIM_Cubic, leaveTangentWeight=1.1106, time=0, value=0}
           {arriveTangentWeight=0.5241, interpMode=RCIM_Cubic, leaveTangentWeight=0.1667, time=0.5, value=0.65}
           <at 1 0 RCIM_Cubic>
          ]}}}
       CurveIndex=$System.Timeline001.NormalizedTime
      )
      Alpha=$Particles.InterpolateOverTime004.InterpolateOverTime_Float.Moving Average
     )
    )) {isSetParameters=true, module=SetVariables_#16}
   &SkeletalMeshLocation(
    Bone Sampling Mode=::NewEnumerator4
    MaskSpawnGroup=false
    Orientation Sampling=::NewEnumerator1
    Sampled Position Offset=&Add_Vector(
     A=$Particles.CatchupVector
     B=&RotateVector(
      Vector To Rotate:Vector3f=&Add_Vector(
       A=<0 0 0>
       B=&Add_Vector(
        A=&SelectVectorFromArray(
         Direct Array Index=&ReturnExecIndex()
         Direct Array Mode=::NewEnumerator1
         Vector Selection Array=DataInterface() {propertyValues {FloatData [<75 0 100>, <-75 0 100>]}}
        )
        B=$System.OverallTargetOffset
       )
      )
      Yaw:NiagaraFloat=&Waveform(
       "Waveform [1]"=::NewEnumerator12
       "[1] Input Position"=&RandomRangeFloat(Minimum=200000, Maximum=100000)
       "[1] Frequency"=1.25
       "[1] Amplitude Min/Max"=<7 -7>
      )
      Pitch:NiagaraFloat=0
      Roll:NiagaraFloat=0
      Rotation Rate:NiagaraFloat=1
     )
    )
    Skeletal Mesh=$System.SkeletalMesh
    Use Sampled Mesh Attribute as Ribbon ID=false
   )
   &ScaleMeshSize(Scale Factor=&VectorFromFloat(Value=&FloatFromCurve(
      FloatCurve=DataInterface() {propertyValues {Curve {keys [{arriveTangent=1.0186, interpMode=RCIM_Cubic, leaveTangent=1.0186, tangentMode=RCTM_User, time=0, value=0}, <at 1 1 RCIM_Cubic>]}}}
      CurveIndex=$Particles.Age
     )))
   &Drag(Drag=2, Use Linear Drag=true, Use Rotational Drag=false)
   &SolveForcesAndVelocity(Clamp Velocity=false, Limit Acceleration=false)
   /SetVariables(Particles.MeshOrientation=$Engine.Owner.Rotation) {enabled=false, isSetParameters=true, module=SetVariables_#17}
   &UpdateMeshOrientation(
    Facing Direction=&Lerp_Vector(A=<5 0 0>, B=<0.5 0 0>, Alpha=$User.ParaGlideLeanAlpha)
    Facing Interpolation=false
    Orientation Method=::NewEnumerator1
    Override Quat=false
    Side Direction:Vector3f=<0 0 0>
    Stabilization Mode:ENiagara_OrientVectorCount=::NewEnumerator3
    TransformFacing=false
    TransformSide=false
    TransformUp=false
    Up Direction:Vector3f=<0 1 0>
   )
   &InterpolateOverTime(
    Rate Of Change:NiagaraFloat=&Lerp_Float(A=2, B=3, Alpha=$System.Dive)
    Target Quat Value:Quat4f=$Particles.MeshOrientation
    Variable Type=::NewEnumerator5
   ) {module=InterpolateOverTime001}
   /SetVariables(Particles.NewOrient:Quat4f=$Particles.InterpolateOverTime001.InterpolateOverTime_Quaternion.Moving Average) {enabled=false, isSetParameters=true, module=SetVariables_#18}
   /SetVariables(Particles.MeshOrientation=$Particles.NewOrient) {enabled=false, isSetParameters=true, module=SetVariables_#19}
   &RotateAroundPoint(
    Radius=110
    Rotation Center:NiagaraPosition=$Particles.SkeletalMeshLocation.SampledPosition
    Rotation Phase:NiagaraFloat=&Add_Float(
     A=$System.Timeline002.NormalizedTime
     B=&SelectFloatFromArray(
      Direct Array Index=&ReturnExecIndex()
      Float Selection Array=DataInterface() {propertyValues {FloatData [0,0.5]}}
     )
    )
    Rotation Rate:NiagaraFloat=&MakeCustomFloatFromBool(Bool=&SetBoolByFloatComparison(A=$System.Timeline002.NormalizedTime, B=0), True Float=0, False Float=1)
    Set Particle Position=false
   )
   /SetVariables(Particles.PrevPosition:NiagaraPosition=$Particles.Position) {isSetParameters=true, module=SetVariables_#20}
   /SetVariables(Particles.Position=&Lerp_Vector(
     A=Linked() {linkedVariable {name=Particles.PrevPosition, type=NiagaraPosition}}
     B=&Add_Vector(A=Linked() {linkedVariable {name=Output.RotateAroundPoint.OutputPosition, type=NiagaraPosition}}, B=<0 0 100>)
     Alpha=&FloatFromCurve(
      FloatCurve=DataInterface() {propertyValues {Curve {defaultValue=1, keys [<at 0 0 RCIM_Cubic>,<at 1 1 RCIM_Cubic>]}}}
      CurveIndex=$System.Timeline003.NormalizedTime
     )
    )) {isSetParameters=true, module=SetVariables_#21}
   /SetVariables(
    Particles.BladesColor:LinearColor=<rgba 0.8073 0.8073 0.8073 2>
    Particles.BladesDynamic:Vector4f=&MakeVector4(X=8, Y=0, Z=0, W=0)
   ) {isSetParameters=true, module=SetVariables_#22}
  ]
  rendererClasses [&NiagaraSpriteRendererProperties, &NiagaraMeshRendererProperties]
  renderers [
   {rendererClass=&NiagaraSpriteRendererProperties}
   {rendererClass=&NiagaraMeshRendererProperties, rendererIndex=1}
  ]
 }
 RightWingTrail {
  emitterUpdateScript [
   &EmitterState()
   &SpawnParticlesFromOtherEmitter(
    Attribute Reader=DataInterface() {propertyValues {EmitterBinding {bindingMode=Other, emitterName=Plane}}}
    Spawn Rate=&MakeCustomFloatFromBool(
     Bool=$User.bIsFrontEndPreview
     True Float=0
     False Float=&Lerp_Float(A=0, B=600, Alpha=&MakeCustomFloatFromBool(Bool=$System.SpeedBool, True Float=1, False Float=0))
    )
    Spawn Rate Per Particle Cap=1000
    Use Maximum Age=false
    Use Minimum Age=false
   )
  ]
  particleSpawnScript [
   &InitializeParticle(Color=<rgba 0.4271 0.4271 0.4271 0.5>, Lifetime=0.01, Sprite Size Mode=::NewEnumerator3, UsePositionOffset=true, Write Material Random=true)
   &SampleParticlesFromOtherEmitter(
    Mesh Orientation Sampling:ENiagara_DirectReadSamplingMode=::NewEnumerator1
    Particle ID Sampling=::NewEnumerator1
    Position Sampling:ENiagara_DirectReadSamplingMode=::NewEnumerator1
   )
   /SetVariables() {isSetParameters=true, module=SetVariables_#11}
   /SetVariables(Particles.RibbonID=$Output.SampleParticlesFromOtherEmitter.SampledID) {isSetParameters=true, module=SetVariables_#12}
   &ShapeLocation(
    Additional Quaternion Rotation=true
    Additional Yaw / Pitch / Roll=<0 90 0>
    Cylinder Distribution=::NewEnumerator2
    Cylinder Height=68
    Cylinder Radius=6
    MaskSpawnGroup=false
    Offset Coordinate Space=::Local
    Offset Mode=::NewEnumerator0
    Rotation Angle Type:ENiagara_AngleInput=::NewEnumerator0
    Rotation Mode=::NewEnumerator3
    Rotation Quaternion=$Output.SampleParticlesFromOtherEmitter.SampledMeshOrientation
    Shape Origin=$Output.SampleParticlesFromOtherEmitter.SampledPosition
    Shape Primitive=::NewEnumerator1
    U Position=&Lerp_Float(A=0.39, B=1, Alpha=0)
    UseOwnerScale=true
    V Position=0
   ) {module=ShapeLocation001}
  ]
  particleUpdateScript [
   &ParticleState()
   &UpdateParticlesFromOtherEmitter(
    Attribute Reader:NiagaraDataInterfaceParticleRead=$Emitter.SpawnParticlesAttributeReader
    Mesh Orientation Sampling:ENiagara_DirectReadSamplingMode=::NewEnumerator1
   ) {enabled=false}
   &ShapeLocation(
    Additional Quaternion Rotation=true
    Additional Yaw / Pitch / Roll=<0 90 0>
    Cylinder Distribution=::NewEnumerator2
    Cylinder Height=68
    Cylinder Radius=7
    MaskSpawnGroup=false
    Offset Coordinate Space=::World
    Offset Mode=::NewEnumerator0
    Rotation Angle Type:ENiagara_AngleInput=::NewEnumerator0
    Rotation Mode=::NewEnumerator3
    Rotation Quaternion=$Output.UpdateParticlesFromOtherEmitter.SampledMeshOrientation
    Shape Origin=$Output.UpdateParticlesFromOtherEmitter.SampledPosition
    Shape Primitive=::NewEnumerator1
    U Position=&Lerp_Float(A=0.4, B=1, Alpha=0)
    UseOwnerScale=true
    V Position=1
   ) {enabled=false}
   &SolveForcesAndVelocity(Clamp Velocity=false, Limit Acceleration=false)
   &ScaleRibbonWidth(Ribbon Width Scale=&FloatFromCurve(FloatCurve=DataInterface() {propertyValues {Curve {keys [<at 0 1>,<at 1 0>]}}}))
  ]
  rendererClasses [&NiagaraRibbonRendererProperties, &NiagaraSpriteRendererProperties]
  renderers [
   {rendererClass=&NiagaraRibbonRendererProperties}
   {rendererClass=&NiagaraSpriteRendererProperties, rendererIndex=1}
  ]
 }
 RightWingTrailFE {
  emitterUpdateScript [
   &EmitterState()
   &SpawnParticlesFromOtherEmitter(
    Attribute Reader=DataInterface() {propertyValues {EmitterBinding {bindingMode=Other, emitterName=PlaneFE}}}
    Spawn Rate=&MakeCustomFloatFromBool(Bool=$User.bIsFrontEndPreview, True Float=&Lerp_Float(A=0, B=600, Alpha=$User.ParaGlideLeanAlpha), False Float=0)
    Spawn Rate Per Particle Cap=1000
    Use Maximum Age=false
    Use Minimum Age=false
   )
  ]
  particleSpawnScript [
   &InitializeParticle(Color=<rgba 0.4271 0.4271 0.4271 0.1>, Lifetime=0.01, Sprite Size Mode=::NewEnumerator3, UsePositionOffset=true, Write Material Random=true)
   &SampleParticlesFromOtherEmitter(
    Mesh Orientation Sampling:ENiagara_DirectReadSamplingMode=::NewEnumerator1
    Particle ID Sampling=::NewEnumerator1
    Position Sampling:ENiagara_DirectReadSamplingMode=::NewEnumerator1
   )
   /SetVariables() {isSetParameters=true, module=SetVariables_#11}
   /SetVariables(Particles.RibbonID=$Output.SampleParticlesFromOtherEmitter.SampledID) {isSetParameters=true, module=SetVariables_#12}
   &ShapeLocation(
    Additional Quaternion Rotation=true
    Additional Yaw / Pitch / Roll=<0 90 0>
    Cylinder Distribution=::NewEnumerator2
    Cylinder Height=68
    Cylinder Radius=6
    MaskSpawnGroup=false
    Offset Coordinate Space=::Local
    Offset Mode=::NewEnumerator0
    Rotation Angle Type:ENiagara_AngleInput=::NewEnumerator0
    Rotation Mode=::NewEnumerator3
    Rotation Quaternion=$Output.SampleParticlesFromOtherEmitter.SampledMeshOrientation
    Shape Origin=$Output.SampleParticlesFromOtherEmitter.SampledPosition
    Shape Primitive=::NewEnumerator1
    U Position=&Lerp_Float(A=0.126, B=1, Alpha=0)
    UseOwnerScale=true
    V Position=0
   ) {module=ShapeLocation001}
   &LinearForce(Force=<0 -1000 3000>)
   &ApplyInitialForces()
  ]
  particleUpdateScript [
   &ParticleState()
   &UpdateParticlesFromOtherEmitter(
    Attribute Reader:NiagaraDataInterfaceParticleRead=$Emitter.SpawnParticlesAttributeReader
    Mesh Orientation Sampling:ENiagara_DirectReadSamplingMode=::NewEnumerator1
   ) {enabled=false}
   &ShapeLocation(
    Additional Quaternion Rotation=true
    Additional Yaw / Pitch / Roll=<0 90 0>
    Cylinder Distribution=::NewEnumerator2
    Cylinder Height=68
    Cylinder Radius=7
    MaskSpawnGroup=false
    Offset Coordinate Space=::World
    Offset Mode=::NewEnumerator0
    Rotation Angle Type:ENiagara_AngleInput=::NewEnumerator0
    Rotation Mode=::NewEnumerator3
    Rotation Quaternion=$Output.UpdateParticlesFromOtherEmitter.SampledMeshOrientation
    Shape Origin=$Output.UpdateParticlesFromOtherEmitter.SampledPosition
    Shape Primitive=::NewEnumerator1
    U Position=&Lerp_Float(A=0.4, B=1, Alpha=0)
    UseOwnerScale=true
    V Position=1
   ) {enabled=false}
   &SolveForcesAndVelocity(Clamp Velocity=false, Limit Acceleration=false)
   &ScaleRibbonWidth(Ribbon Width Scale=&FloatFromCurve(FloatCurve=DataInterface() {propertyValues {Curve {keys [<at 0 1>,<at 1 0>]}}}))
  ]
  rendererClasses [&NiagaraRibbonRendererProperties, &NiagaraSpriteRendererProperties]
  renderers [
   {rendererClass=&NiagaraRibbonRendererProperties}
   {rendererClass=&NiagaraSpriteRendererProperties, rendererIndex=1}
  ]
 }
 SmokeBurst {
  bEnabled=false
  emitterUpdateScript [
   &EmitterState()
   &SpawnParticlesFromOtherEmitter(
    Attribute Reader=DataInterface() {propertyValues {EmitterBinding {bindingMode=Other, emitterName=Plane}}}
    Spawn Rate=&MakeCustomFloatFromBool(
     Bool=$User.bIsFrontEndPreview
     True Float=0
     False Float=&Lerp_Float(A=100, B=200, Alpha=$User.ParaGlideLeanAlpha)
    )
    Spawn Rate Per Particle Cap=1000
    Use Maximum Age=false
    Use Minimum Age=false
   )
  ]
  particleSpawnScript [
   &InitializeParticle(
    Color=<rgba 1 1 1 0.5>
    Lifetime=&Lerp_Float(A=0.15, B=0.25, Alpha=$User.ParaGlideLeanAlpha)
    Sprite Rotation Mode=::NewEnumerator2
    Sprite Size Mode=::NewEnumerator3
    Uniform Sprite Size=30
    UsePositionOffset=false
    Write Material Random=true
   )
   &SampleParticlesFromOtherEmitter(Mesh Orientation Sampling:ENiagara_DirectReadSamplingMode=::NewEnumerator1, Particle ID Sampling=::NewEnumerator1)
   &ShapeLocation(
    Additional Quaternion Rotation=true
    Additional Yaw / Pitch / Roll=<0 90 90>
    Cylinder Distribution=::NewEnumerator2
    Cylinder Height=37
    Cylinder Radius=2.4
    MaskSpawnGroup=false
    Offset Mode=::NewEnumerator0
    Rotation Angle Type:ENiagara_AngleInput=::NewEnumerator0
    Rotation Mode=::NewEnumerator3
    Rotation Quaternion=$Output.SampleParticlesFromOtherEmitter.SampledMeshOrientation
    Shape Origin=$Output.SampleParticlesFromOtherEmitter.SampledPosition
    Shape Primitive=::NewEnumerator1
    U Position=&Lerp_Float(A=0.5, B=0.5, Alpha=1)
    UseOwnerScale=true
    V Position=0
   )
  ]
  particleUpdateScript [
   &ParticleState()
   &ScaleSpriteSize(
    Uniform Curve Scale=&Lerp_Float(A=1, B=3, Alpha=$System.Timeline003.NormalizedTime)
    Uniform Curve Sprite Scale=DataInterface() {propertyValues {Curve {
       defaultValue=1
       keys [
        {arriveTangent=0.5183, interpMode=RCIM_Cubic, leaveTangent=0.5183, tangentMode=RCTM_User, time=0, value=0.1}
        {interpMode=RCIM_Cubic, leaveTangent=0.8555, time=0.25, value=1}
       ]
      }}}
   )
   &SolveForcesAndVelocity(Clamp Velocity=false, Limit Acceleration=false)
   &ScaleColor(
    Scale Alpha=&FloatFromCurve(FloatCurve=DataInterface() {propertyValues {Curve {defaultValue=1, keys [<at 0 0>,<at 0.025 1>,<at 1 0>]}}})
    Scale RGB=&MakeVectorFromLinearColorRGB(Color=&ColorFromCurve(ColorCurve=DataInterface() {propertyValues {
        AlphaCurve {keys [<at 0 1 RCIM_Cubic>]}
        BlueCurve {keys [<at 0.182 0 RCIM_Cubic>,<at 0.774 0.0152 RCIM_Cubic>]}
        GreenCurve {keys [<at 0.182 0.2448 RCIM_Cubic>,<at 0.774 0.5333 RCIM_Cubic>]}
        RedCurve {keys [<at 0.182 0.0326 RCIM_Cubic>,<at 0.774 0 RCIM_Cubic>]}
       }}))
    ScaleA=true
    ScaleRGB=true
   )
   &DynamicMaterialParameters(
    Index 0 Param 1=&FloatFromCurve(FloatCurve=DataInterface() {propertyValues {Curve {defaultValue=1, keys [{tangentMode=RCTM_User, time=0, value=1}, <at 0.1387 0>, {tangentMode=RCTM_User, time=1, value=1}]}}})
    Index 0 Param 2=1
    Index 0 Param 3=1
    Index 1 Param 1=10
    Index 1 Param 2=100
    Module.Index0Float0Write=true
    Module.Index0Float1Write=true
    Module.Index0Float2Write=true
    Module.Index0Float3Write=true
    Module.Index1Float0Write=true
    Module.Index1Float1Write=true
    Module.Index1Float2Write=true
    Module.Index1Float3Write=true
    Write Parameter Index 1=::NewEnumerator1
   )
   &ScaleColor(
    Scale Alpha=&FloatFromCurve(
     FloatCurve=DataInterface() {propertyValues {Curve {defaultValue=1, keys [{tangentMode=RCTM_User, time=0, value=0}, {tangentMode=RCTM_User, time=1, value=1}]}}}
     CurveIndex=$System.Timeline001.NormalizedTime
    )
    Scale RGB=&MakeVectorFromLinearColorRGB(Color=$System.TOD)
    ScaleA=true
    ScaleRGB=true
   ) {module=ScaleColor001}
  ]
  rendererClasses [&NiagaraSpriteRendererProperties]
  renderers [{rendererClass=&NiagaraSpriteRendererProperties}]
 }
 SmokeTrail {
  bEnabled=false
  emitterUpdateScript [
   &EmitterState()
   &SpawnParticlesFromOtherEmitter(
    Attribute Reader=DataInterface() {propertyValues {EmitterBinding {bindingMode=Other, emitterName=Plane}}}
    Spawn Rate=&MakeCustomFloatFromBool(
     Bool=$User.bIsFrontEndPreview
     True Float=0
     False Float=&Lerp_Float(A=200, B=400, Alpha=$User.ParaGlideLeanAlpha)
    )
    Spawn Rate Per Particle Cap=1000
    Use Maximum Age=false
    Use Minimum Age=false
   )
  ]
  particleSpawnScript [
   &InitializeParticle(
    Lifetime=&Lerp_Float(A=0.13, B=0.25, Alpha=$System.Timeline003.NormalizedTime)
    Sprite Size Mode=::NewEnumerator3
    Uniform Sprite Size=2
    UsePositionOffset=false
    Write Material Random=true
   )
   &SampleParticlesFromOtherEmitter(Mesh Orientation Sampling:ENiagara_DirectReadSamplingMode=::NewEnumerator1, Particle ID Sampling=::NewEnumerator1)
   /SetVariables(Particles.RibbonWidth=&Lerp_Float(A=30, B=80, Alpha=$User.ParaGlideLeanAlpha)) {isSetParameters=true, module=SetVariables_#11}
   /SetVariables(Particles.RibbonID=&self:RealmTrailID()) {isSetParameters=true, module=SetVariables_#12}
   /SetVariables(Particles.RibbonLinkOrder:NiagaraFloat=$Output.SampleParticlesFromOtherEmitter.SampledAge) {enabled=false, isSetParameters=true, module=SetVariables_#23}
   &ShapeLocation(
    Additional Quaternion Rotation=true
    Additional Yaw / Pitch / Roll=<0 90 90>
    Cylinder Distribution=::NewEnumerator2
    Cylinder Height=37
    Cylinder Radius=&Lerp_Float(A=1.5, B=3.5, Alpha=$User.ParaGlideLeanAlpha)
    MaskSpawnGroup=false
    Offset Mode=::NewEnumerator0
    Rotation Angle Type:ENiagara_AngleInput=::NewEnumerator0
    Rotation Mode=::NewEnumerator3
    Rotation Quaternion=$Output.SampleParticlesFromOtherEmitter.SampledMeshOrientation
    Shape Origin=$Output.SampleParticlesFromOtherEmitter.SampledPosition
    Shape Primitive=::NewEnumerator1
    U Position=&Lerp_Float(A=0.5, B=0.5, Alpha=1)
    UseOwnerScale=true
    V Position=&Lerp_Float(A=0.05, B=0.25, Alpha=$User.ParaGlideLeanAlpha)
   )
  ]
  particleUpdateScript [
   &ParticleState()
   &SolveForcesAndVelocity(Clamp Velocity=false, Limit Acceleration=false)
   &ScaleColor(
    Scale Alpha=&FloatFromCurve(FloatCurve=DataInterface() {propertyValues {Curve {defaultValue=1, keys [<at 0 0>,<at 0.025 1>,<at 1 0>]}}})
    Scale RGB=&MakeVectorFromLinearColorRGB(Color=&ColorFromCurve(ColorCurve=DataInterface() {propertyValues {
        AlphaCurve {keys [<at 0 1 RCIM_Cubic>]}
        BlueCurve {keys [<at 0.182 0 RCIM_Cubic>,<at 0.774 0.0152 RCIM_Cubic>]}
        GreenCurve {keys [<at 0.182 0.1667 RCIM_Cubic>,<at 0.774 0.5333 RCIM_Cubic>]}
        RedCurve {keys [<at 0.182 0.0222 RCIM_Cubic>,<at 0.774 0 RCIM_Cubic>]}
       }}))
    ScaleA=true
    ScaleRGB=true
   )
   &DynamicMaterialParameters(
    Index 0 Param 1=-2.5
    Index 0 Param 4=&FloatFromCurve(FloatCurve=DataInterface() {propertyValues {Curve {keys [<at 0 0>,<at 0.01 1>,<at 0.5 1>,<at 1 0>]}}})
    Module.Index0Float0Write=true
    Module.Index0Float1Write=true
    Module.Index0Float2Write=true
    Module.Index0Float3Write=true
   )
   &ScaleRibbonWidth(Ribbon Width Scale=&FloatFromCurve(FloatCurve=DataInterface() {propertyValues {Curve {
        defaultValue=1
        keys [
         {arriveTangent=0.5183, interpMode=RCIM_Cubic, leaveTangent=0.5183, tangentMode=RCTM_User, time=0, value=0.025}
         {interpMode=RCIM_Cubic, leaveTangent=0.8555, time=0.25, value=1}
        ]
       }}}))
   &ScaleColor(
    Scale Alpha=&FloatFromCurve(
     FloatCurve=DataInterface() {propertyValues {Curve {defaultValue=1, keys [{tangentMode=RCTM_User, time=0, value=0}, {tangentMode=RCTM_User, time=1, value=1}]}}}
     CurveIndex=$System.Timeline001.NormalizedTime
    )
    Scale RGB=&MakeVectorFromLinearColorRGB(Color=&FNTimeOfDayColor(Night Color=<rgba 0.5713 0.5615 0.7 1>, Morning Color=<rgba 1 1 1 1>, Day Color=<rgba 1 1 1 1>, Evening Color=<rgba 1 1 1 1>))
    ScaleA=true
    ScaleRGB=true
   ) {module=ScaleColor001}
  ]
  rendererClasses [&NiagaraSpriteRendererProperties, &NiagaraRibbonRendererProperties]
  renderers [
   {rendererClass=&NiagaraSpriteRendererProperties}
   {rendererClass=&NiagaraRibbonRendererProperties, rendererIndex=1}
  ]
 }
 SmokeTrail2 {
  bEnabled=false
  emitterUpdateScript [
   &EmitterState()
   &SpawnParticlesFromOtherEmitter(
    Attribute Reader=DataInterface() {propertyValues {EmitterBinding {bindingMode=Other, emitterName=Plane}}}
    Spawn Rate=&MakeCustomFloatFromBool(Bool=$User.bIsFrontEndPreview, True Float=0, False Float=200)
    Spawn Rate Per Particle Cap=1000
    Use Maximum Age=false
    Use Minimum Age=false
   )
  ]
  particleSpawnScript [
   &InitializeParticle(
    Lifetime=&Lerp_Float(A=0.15, B=0.25, Alpha=$System.Timeline003.NormalizedTime)
    Sprite Size Mode=::NewEnumerator3
    UsePositionOffset=false
    Write Material Random=true
   )
   &SampleParticlesFromOtherEmitter(Mesh Orientation Sampling:ENiagara_DirectReadSamplingMode=::NewEnumerator1, Particle ID Sampling=::NewEnumerator1)
   /SetVariables(Particles.RibbonWidth=&Lerp_Float(A=30, B=80, Alpha=$User.ParaGlideLeanAlpha)) {isSetParameters=true, module=SetVariables_#11}
   /SetVariables(Particles.RibbonID=&self:RealmTrailID()) {isSetParameters=true, module=SetVariables_#12}
   /SetVariables(Particles.RibbonLinkOrder:NiagaraFloat=$Output.SampleParticlesFromOtherEmitter.SampledAge) {enabled=false, isSetParameters=true, module=SetVariables_#23}
   &ShapeLocation(
    Additional Quaternion Rotation=true
    Additional Yaw / Pitch / Roll=<0 90 90>
    Cylinder Distribution=::NewEnumerator2
    Cylinder Height=37
    Cylinder Radius=&Lerp_Float(A=1.5, B=3.5, Alpha=$User.ParaGlideLeanAlpha)
    MaskSpawnGroup=false
    Offset Mode=::NewEnumerator0
    Rotation Angle Type:ENiagara_AngleInput=::NewEnumerator0
    Rotation Mode=::NewEnumerator3
    Rotation Quaternion=$Output.SampleParticlesFromOtherEmitter.SampledMeshOrientation
    Shape Origin=$Output.SampleParticlesFromOtherEmitter.SampledPosition
    Shape Primitive=::NewEnumerator1
    U Position=&Lerp_Float(A=0.5, B=0.5, Alpha=1)
    UseOwnerScale=true
    V Position=&Lerp_Float(A=0.05, B=0.25, Alpha=$User.ParaGlideLeanAlpha)
   )
  ]
  particleUpdateScript [
   &ParticleState()
   &SolveForcesAndVelocity(Clamp Velocity=false, Limit Acceleration=false)
   &ScaleColor(
    Scale Alpha=&FloatFromCurve(FloatCurve=DataInterface() {propertyValues {Curve {defaultValue=1, keys [<at 0 0>,<at 0.025 1>,<at 1 0>]}}})
    Scale RGB=&MakeVectorFromLinearColorRGB(Color=&ColorFromCurve(ColorCurve=DataInterface() {propertyValues {
        AlphaCurve {keys [<at 0 1 RCIM_Cubic>]}
        BlueCurve {keys [<at 0.182 0 RCIM_Cubic>,<at 0.774 0.0152 RCIM_Cubic>]}
        GreenCurve {keys [<at 0.182 0.1667 RCIM_Cubic>,<at 0.774 0.5333 RCIM_Cubic>]}
        RedCurve {keys [<at 0.182 0.0222 RCIM_Cubic>,<at 0.774 0 RCIM_Cubic>]}
       }}))
    ScaleA=true
    ScaleRGB=true
   )
   &DynamicMaterialParameters(Module.Index0Float0Write=true, Module.Index0Float1Write=true, Module.Index0Float2Write=true, Module.Index0Float3Write=true) {enabled=false}
   &ScaleRibbonWidth(Ribbon Width Scale=&FloatFromCurve(FloatCurve=DataInterface() {propertyValues {Curve {
        defaultValue=1
        keys [
         {arriveTangent=0.5183, interpMode=RCIM_Cubic, leaveTangent=0.5183, tangentMode=RCTM_User, time=0, value=0.025}
         {interpMode=RCIM_Cubic, leaveTangent=0.8555, time=0.25, value=1}
        ]
       }}}))
   &ScaleColor(
    Scale Alpha=&FloatFromCurve(
     FloatCurve=DataInterface() {propertyValues {Curve {defaultValue=1, keys [{tangentMode=RCTM_User, time=0, value=0}, {tangentMode=RCTM_User, time=1, value=1}]}}}
     CurveIndex=$System.Timeline001.NormalizedTime
    )
    Scale RGB=&MakeVectorFromLinearColorRGB(Color=$System.TOD)
    ScaleA=true
    ScaleRGB=true
   ) {module=ScaleColor001}
  ]
  rendererClasses [&NiagaraSpriteRendererProperties, &NiagaraRibbonRendererProperties]
  renderers [
   {rendererClass=&NiagaraSpriteRendererProperties}
   {rendererClass=&NiagaraRibbonRendererProperties, rendererIndex=1}
  ]
 }
 SmokeTrail2FE {
  bEnabled=false
  emitterUpdateScript [
   &EmitterState()
   &SpawnParticlesFromOtherEmitter(
    Attribute Reader=DataInterface() {propertyValues {EmitterBinding {bindingMode=Other, emitterName=PlaneFE}}}
    Spawn Rate=&MakeCustomFloatFromBool(Bool=$User.bIsFrontEndPreview, True Float=200, False Float=0)
    Spawn Rate Per Particle Cap=1000
    Use Maximum Age=false
    Use Minimum Age=false
   )
  ]
  particleSpawnScript [
   &InitializeParticle(
    Lifetime=&Lerp_Float(A=0.15, B=0.25, Alpha=$System.Timeline003.NormalizedTime)
    Sprite Size Mode=::NewEnumerator3
    UsePositionOffset=false
    Write Material Random=true
   )
   &SampleParticlesFromOtherEmitter(Mesh Orientation Sampling:ENiagara_DirectReadSamplingMode=::NewEnumerator1, Particle ID Sampling=::NewEnumerator1)
   /SetVariables(Particles.RibbonWidth=&Lerp_Float(A=30, B=60, Alpha=$User.ParaGlideLeanAlpha)) {isSetParameters=true, module=SetVariables_#11}
   /SetVariables(Particles.RibbonID=&self:RealmTrailID()) {isSetParameters=true, module=SetVariables_#12}
   /SetVariables(Particles.RibbonLinkOrder:NiagaraFloat=$Output.SampleParticlesFromOtherEmitter.SampledAge) {enabled=false, isSetParameters=true, module=SetVariables_#23}
   &ShapeLocation(
    Additional Quaternion Rotation=true
    Additional Yaw / Pitch / Roll=<0 90 90>
    Cylinder Distribution=::NewEnumerator2
    Cylinder Height=37
    Cylinder Radius=2
    MaskSpawnGroup=false
    Offset Mode=::NewEnumerator0
    Rotation Angle Type:ENiagara_AngleInput=::NewEnumerator0
    Rotation Mode=::NewEnumerator3
    Rotation Quaternion=$Output.SampleParticlesFromOtherEmitter.SampledMeshOrientation
    Shape Origin=$Output.SampleParticlesFromOtherEmitter.SampledPosition
    Shape Primitive=::NewEnumerator1
    U Position=&Lerp_Float(A=0, B=0, Alpha=1)
    UseOwnerScale=true
    V Position=0
   )
   &LinearForce(Force=&Lerp_Vector(A=<0 -300 1000>, B=<0 -1000 2000>, Alpha=$User.ParaGlideLeanAlpha)) {module=LinearForce001}
   &ApplyInitialForces()
  ]
  particleUpdateScript [
   &ParticleState()
   &SolveForcesAndVelocity(Clamp Velocity=false, Limit Acceleration=false)
   &ScaleColor(
    Scale Alpha=&FloatFromCurve(FloatCurve=DataInterface() {propertyValues {Curve {defaultValue=1, keys [<at 0 0>,<at 0.025 1>,<at 1 0>]}}})
    Scale RGB=&MakeVectorFromLinearColorRGB(Color=&ColorFromCurve(ColorCurve=DataInterface() {propertyValues {
        AlphaCurve {keys [<at 0 1 RCIM_Cubic>]}
        BlueCurve {keys [<at 0.182 0 RCIM_Cubic>,<at 0.774 0.0152 RCIM_Cubic>]}
        GreenCurve {keys [<at 0.182 0.1667 RCIM_Cubic>,<at 0.774 0.5333 RCIM_Cubic>]}
        RedCurve {keys [<at 0.182 0.0222 RCIM_Cubic>,<at 0.774 0 RCIM_Cubic>]}
       }}))
    ScaleA=true
    ScaleRGB=true
   )
   &DynamicMaterialParameters(Module.Index0Float0Write=true, Module.Index0Float1Write=true, Module.Index0Float2Write=true, Module.Index0Float3Write=true) {enabled=false}
   &ScaleRibbonWidth(Ribbon Width Scale=&FloatFromCurve(FloatCurve=DataInterface() {propertyValues {Curve {
        defaultValue=1
        keys [
         {arriveTangent=0.5183, interpMode=RCIM_Cubic, leaveTangent=0.5183, tangentMode=RCTM_User, time=0, value=0.1}
         {interpMode=RCIM_Cubic, leaveTangent=0.8555, time=0.25, value=1}
        ]
       }}}))
   &ScaleColor(
    Scale Alpha=&FloatFromCurve(
     FloatCurve=DataInterface() {propertyValues {Curve {defaultValue=1, keys [{tangentMode=RCTM_User, time=0, value=0}, {tangentMode=RCTM_User, time=1, value=1}]}}}
     CurveIndex=$System.Timeline001.NormalizedTime
    )
    ScaleA=true
    ScaleRGB=true
   ) {enabled=false, module=ScaleColor001}
  ]
  rendererClasses [&NiagaraSpriteRendererProperties, &NiagaraRibbonRendererProperties]
  renderers [
   {rendererClass=&NiagaraSpriteRendererProperties}
   {rendererClass=&NiagaraRibbonRendererProperties, rendererIndex=1}
  ]
 }
 SmokeTrailCombined {
  emitterUpdateScript [
   &EmitterState()
   &SpawnParticlesFromOtherEmitter(
    Attribute Reader=DataInterface() {propertyValues {EmitterBinding {bindingMode=Other, emitterName=Plane}}}
    Spawn Rate=&MakeCustomFloatFromBool(
     Bool=$User.bIsFrontEndPreview
     True Float=0
     False Float=&Lerp_Float(A=200, B=400, Alpha=$User.ParaGlideLeanAlpha)
    )
    Spawn Rate Per Particle Cap=1000
    Use Maximum Age=false
    Use Minimum Age=false
   )
  ]
  particleSpawnScript [
   &InitializeParticle(
    Lifetime=&Lerp_Float(A=0.13, B=0.25, Alpha=$System.Timeline003.NormalizedTime)
    Sprite Rotation Mode=::NewEnumerator2
    Sprite Size Mode=::NewEnumerator3
    Uniform Sprite Size=30
    UsePositionOffset=false
    Write Material Random=true
   )
   &SampleParticlesFromOtherEmitter(Mesh Orientation Sampling:ENiagara_DirectReadSamplingMode=::NewEnumerator1, Particle ID Sampling=::NewEnumerator1)
   /SetVariables(Particles.RibbonWidth=&Lerp_Float(A=30, B=80, Alpha=$User.ParaGlideLeanAlpha)) {isSetParameters=true, module=SetVariables_#11}
   /SetVariables(Particles.RibbonID=&self:RealmTrailID()) {isSetParameters=true, module=SetVariables_#12}
   /SetVariables(Particles.RibbonLinkOrder:NiagaraFloat=$Output.SampleParticlesFromOtherEmitter.SampledAge) {enabled=false, isSetParameters=true, module=SetVariables_#23}
   &ShapeLocation(
    Additional Quaternion Rotation=true
    Additional Yaw / Pitch / Roll=<0 90 90>
    Cylinder Distribution=::NewEnumerator2
    Cylinder Height=37
    Cylinder Radius=&Lerp_Float(A=0.1, B=3.5, Alpha=$User.ParaGlideLeanAlpha)
    MaskSpawnGroup=false
    Offset Mode=::NewEnumerator0
    Rotation Angle Type:ENiagara_AngleInput=::NewEnumerator0
    Rotation Mode=::NewEnumerator3
    Rotation Quaternion=$Output.SampleParticlesFromOtherEmitter.SampledMeshOrientation
    Shape Origin=$Output.SampleParticlesFromOtherEmitter.SampledPosition
    Shape Primitive=::NewEnumerator1
    U Position=&Lerp_Float(A=0.5, B=0.5, Alpha=1)
    UseOwnerScale=true
    V Position=&Lerp_Float(A=0.075, B=0.25, Alpha=$User.ParaGlideLeanAlpha)
   )
  ]
  particleUpdateScript [
   &ParticleState()
   &SolveForcesAndVelocity(Clamp Velocity=false, Limit Acceleration=false)
   &ScaleColor(
    Scale Alpha=&FloatFromCurve(FloatCurve=DataInterface() {propertyValues {Curve {defaultValue=1, keys [<at 0 0>,<at 0.025 1>,<at 1 0>]}}})
    Scale RGB=&MakeVectorFromLinearColorRGB(Color=&ColorFromCurve(ColorCurve=DataInterface() {propertyValues {
        AlphaCurve {keys [<at 0 1 RCIM_Cubic>]}
        BlueCurve {keys [<at 0.182 0 RCIM_Cubic>,<at 0.774 0.0152 RCIM_Cubic>]}
        GreenCurve {keys [<at 0.182 0.1667 RCIM_Cubic>,<at 0.774 0.5333 RCIM_Cubic>]}
        RedCurve {keys [<at 0.182 0.0222 RCIM_Cubic>,<at 0.774 0 RCIM_Cubic>]}
       }}))
    ScaleA=true
    ScaleRGB=true
   )
   &DynamicMaterialParameters(
    Index 0 Param 1=-2.5
    Index 0 Param 4=&FloatFromCurve(FloatCurve=DataInterface() {propertyValues {Curve {keys [<at 0 0>,<at 0.01 1>,<at 0.5 1>,<at 1 0>]}}})
    Module.Index0Float0Write=true
    Module.Index0Float1Write=true
    Module.Index0Float2Write=true
    Module.Index0Float3Write=true
   ) {enabled=false}
   &ScaleSpriteSize(
    Uniform Curve Scale=&Lerp_Float(A=1, B=3, Alpha=$System.Timeline003.NormalizedTime)
    Uniform Curve Sprite Scale=DataInterface() {propertyValues {Curve {
       defaultValue=1
       keys [
        {arriveTangent=0.5183, interpMode=RCIM_Cubic, leaveTangent=0.5183, tangentMode=RCTM_User, time=0, value=0.1}
        {interpMode=RCIM_Cubic, leaveTangent=0.8555, time=0.4, value=1}
       ]
      }}}
   )
   /SetVariables(Particles.Smoke1Dyn:Vector4f=&MakeVector4(
     X=-2.5
     Y=0
     Z=0
     W=&FloatFromCurve(FloatCurve=DataInterface() {propertyValues {Curve {keys [<at 0 0>,<at 0.01 1>,<at 0.5 1>,<at 1 0>]}}})
    )) {isSetParameters=true, module=SetVariables_#24}
   /SetVariables(
    Particles.ParticleDyn:Vector4f=&MakeVector4(
     X=&FloatFromCurve(FloatCurve=DataInterface() {propertyValues {Curve {defaultValue=1, keys [{tangentMode=RCTM_User, time=0, value=1}, <at 0.1387 0>, {tangentMode=RCTM_User, time=1, value=1}]}}})
     Y=1
     Z=1
     W=0
    )
    Particles.ParticleDyn2:Vector4f=&MakeVector4(X=10, Y=100, Z=0, W=0)
   ) {isSetParameters=true}
   /SetVariables(Particles.ParticleColor:LinearColor=&MakeLinearColorFromVectorAndFloat(
     "Vector (RGB)"=&MakeVectorFromLinearColorRGB(Color=&ColorFromCurve(ColorCurve=DataInterface() {propertyValues {
         AlphaCurve {keys [<at 0 1 RCIM_Cubic>]}
         BlueCurve {keys [<at 0.1821 0 RCIM_Cubic>,<at 0.774 0.0152 RCIM_Cubic>]}
         GreenCurve {keys [<at 0.1821 0.2448 RCIM_Cubic>,<at 0.774 0.5333 RCIM_Cubic>]}
         RedCurve {keys [<at 0.1821 0.0326 RCIM_Cubic>,<at 0.774 0 RCIM_Cubic>]}
        }}))
     "Float (Alpha)"=&FloatFromCurve(FloatCurve=DataInterface() {propertyValues {Curve {defaultValue=1, keys [<at 0 0>,<at 0.025 1>,<at 1 0>]}}}, Scale Curve=0.25)
    )) {isSetParameters=true}
   /SetVariables(Particles.Smoke2Color:LinearColor=&MakeLinearColorFromVectorAndFloat(
     "Vector (RGB)"=&Multiply_Vector(
      A=&MakeVectorFromLinearColorRGB(Color=$Particles.Color)
      B=&MakeVectorFromLinearColorRGB(Color=$System.TOD)
     )
     "Float (Alpha)"=&Multiply_Float(
      A=&MakeFloatFromLinearColor(LinearColor=$Particles.Color, Channel:ENiagaraFloatFromLinearColorOptions=::NewEnumerator3)
      B=&FloatFromCurve(
       FloatCurve=DataInterface() {propertyValues {Curve {defaultValue=1, keys [{tangentMode=RCTM_User, time=0, value=0}, {tangentMode=RCTM_User, time=1, value=0.5}]}}}
       CurveIndex=$System.Timeline001.NormalizedTime
       Scale Curve=&Lerp_Float(A=1, B=2, Alpha=$User.ParaGlideLeanAlpha)
      )
     )
    )) {isSetParameters=true, module=SetVariables_#25}
   /SetVariables(Particles.ParticleTOD:LinearColor=&MakeLinearColorFromVectorAndFloat(
     "Vector (RGB)"=&Multiply_Vector(
      A=&MakeVectorFromLinearColorRGB(Color=$Particles.ParticleColor)
      B=&MakeVectorFromLinearColorRGB(Color=$System.TOD)
     )
     "Float (Alpha)"=&Multiply_Float(
      A=&MakeFloatFromLinearColor(LinearColor=$Particles.ParticleColor, Channel:ENiagaraFloatFromLinearColorOptions=::NewEnumerator3)
      B=&FloatFromCurve(
       FloatCurve=DataInterface() {propertyValues {Curve {defaultValue=1, keys [{tangentMode=RCTM_User, time=0, value=0}, {tangentMode=RCTM_User, time=1, value=1}]}}}
       CurveIndex=$System.Timeline001.NormalizedTime
       Scale Curve=&Lerp_Float(A=1.5, B=1, Alpha=$User.ParaGlideLeanAlpha)
      )
     )
    )) {isSetParameters=true}
   &ScaleRibbonWidth(Ribbon Width Scale=&Lerp_Float(
     A=&FloatFromCurve(FloatCurve=DataInterface() {propertyValues {Curve {
         defaultValue=1
         keys [
          {arriveTangent=0.5183, interpMode=RCIM_Cubic, leaveTangent=0.5183, tangentMode=RCTM_User, time=0, value=0.025}
          {interpMode=RCIM_Cubic, leaveTangent=0.8555, time=0.4, value=1}
         ]
        }}})
     B=&FloatFromCurve(FloatCurve=DataInterface() {propertyValues {Curve {
         defaultValue=1
         keys [
          {arriveTangent=0.5183, interpMode=RCIM_Cubic, leaveTangent=0.5183, tangentMode=RCTM_User, time=0, value=0.025}
          {interpMode=RCIM_Cubic, leaveTangent=0.8555, time=0.25, value=1}
         ]
        }}})
     Alpha=$User.ParaGlideLeanAlpha
    ))
   &ScaleColor(
    Scale Alpha=&FloatFromCurve(
     FloatCurve=DataInterface() {propertyValues {Curve {defaultValue=1, keys [{tangentMode=RCTM_User, time=0, value=0}, {tangentMode=RCTM_User, time=1, value=1}]}}}
     CurveIndex=$System.Timeline001.NormalizedTime
    )
    Scale RGB=&MakeVectorFromLinearColorRGB(Color=&FNTimeOfDayColor(Night Color=<rgba 0.5713 0.5615 0.7 1>, Morning Color=<rgba 1 1 1 1>, Day Color=<rgba 1 1 1 1>, Evening Color=<rgba 1 1 1 1>))
    ScaleA=true
    ScaleRGB=true
   ) {module=ScaleColor001}
  ]
  rendererClasses [&NiagaraSpriteRendererProperties, &NiagaraRibbonRendererProperties]
  renderers [
   {rendererClass=&NiagaraSpriteRendererProperties}
   {rendererClass=&NiagaraRibbonRendererProperties, rendererIndex=1}
   {rendererClass=&NiagaraRibbonRendererProperties, rendererIndex=2}
   {rendererClass=&NiagaraSpriteRendererProperties, rendererIndex=3}
  ]
 }
 SmokeTrailFE001 {
  bEnabled=false
  emitterUpdateScript [
   &EmitterState()
   &SpawnParticlesFromOtherEmitter(
    Attribute Reader=DataInterface() {propertyValues {EmitterBinding {bindingMode=Other, emitterName=PlaneFE}}}
    Spawn Rate=&MakeCustomFloatFromBool(
     Bool=$User.bIsFrontEndPreview
     True Float=&Lerp_Float(A=200, B=400, Alpha=$User.ParaGlideLeanAlpha)
     False Float=0
    )
    Spawn Rate Per Particle Cap=1000
    Use Maximum Age=false
    Use Minimum Age=false
   )
  ]
  particleSpawnScript [
   &InitializeParticle(
    Lifetime=&Lerp_Float(A=0.13, B=0.25, Alpha=$System.Timeline003.NormalizedTime)
    Sprite Size Mode=::NewEnumerator3
    UsePositionOffset=false
    Write Material Random=true
   )
   &SampleParticlesFromOtherEmitter(Mesh Orientation Sampling:ENiagara_DirectReadSamplingMode=::NewEnumerator1, Particle ID Sampling=::NewEnumerator1)
   /SetVariables(Particles.RibbonWidth=&Lerp_Float(A=30, B=60, Alpha=$User.ParaGlideLeanAlpha)) {isSetParameters=true, module=SetVariables_#11}
   /SetVariables(Particles.RibbonID=&self:RealmTrailID()) {isSetParameters=true, module=SetVariables_#12}
   /SetVariables(Particles.RibbonLinkOrder:NiagaraFloat=$Output.SampleParticlesFromOtherEmitter.SampledAge) {enabled=false, isSetParameters=true, module=SetVariables_#23}
   &ShapeLocation(
    Additional Quaternion Rotation=true
    Additional Yaw / Pitch / Roll=<0 90 90>
    Cylinder Distribution=::NewEnumerator2
    Cylinder Height=37
    Cylinder Radius=2
    MaskSpawnGroup=false
    Offset Mode=::NewEnumerator0
    Rotation Angle Type:ENiagara_AngleInput=::NewEnumerator0
    Rotation Mode=::NewEnumerator3
    Rotation Quaternion=$Output.SampleParticlesFromOtherEmitter.SampledMeshOrientation
    Shape Origin=$Output.SampleParticlesFromOtherEmitter.SampledPosition
    Shape Primitive=::NewEnumerator1
    U Position=&Lerp_Float(A=0, B=0, Alpha=1)
    UseOwnerScale=true
    V Position=0
   )
   &LinearForce(Force=&Lerp_Vector(A=<0 -300 1000>, B=<0 -1000 2000>, Alpha=$User.ParaGlideLeanAlpha)) {module=LinearForce001}
   &ApplyInitialForces()
  ]
  particleUpdateScript [
   &ParticleState()
   &SolveForcesAndVelocity(Clamp Velocity=false, Limit Acceleration=false)
   &ScaleColor(
    Scale Alpha=&FloatFromCurve(FloatCurve=DataInterface() {propertyValues {Curve {defaultValue=1, keys [<at 0 0>,<at 0.025 1>,<at 1 0>]}}})
    Scale RGB=&MakeVectorFromLinearColorRGB(Color=&ColorFromCurve(ColorCurve=DataInterface() {propertyValues {
        AlphaCurve {keys [<at 0 1 RCIM_Cubic>]}
        BlueCurve {keys [<at 0.182 0 RCIM_Cubic>,<at 0.774 0.0152 RCIM_Cubic>]}
        GreenCurve {keys [<at 0.182 0.1667 RCIM_Cubic>,<at 0.774 0.5333 RCIM_Cubic>]}
        RedCurve {keys [<at 0.182 0.0222 RCIM_Cubic>,<at 0.774 0 RCIM_Cubic>]}
       }}))
    ScaleA=true
    ScaleRGB=true
   )
   &DynamicMaterialParameters(
    Index 0 Param 1=-2.5
    Index 0 Param 4=&FloatFromCurve(FloatCurve=DataInterface() {propertyValues {Curve {keys [<at 0 0>,<at 0.01 1>,<at 0.5 1>,<at 1 0>]}}})
    Module.Index0Float0Write=true
    Module.Index0Float1Write=true
    Module.Index0Float2Write=true
    Module.Index0Float3Write=true
   )
   &ScaleRibbonWidth(Ribbon Width Scale=&FloatFromCurve(FloatCurve=DataInterface() {propertyValues {Curve {
        defaultValue=1
        keys [
         {arriveTangent=0.5183, interpMode=RCIM_Cubic, leaveTangent=0.5183, tangentMode=RCTM_User, time=0, value=0.1}
         {interpMode=RCIM_Cubic, leaveTangent=0.8555, time=0.25, value=1}
        ]
       }}}))
   &ScaleColor(
    Scale Alpha=&FloatFromCurve(
     FloatCurve=DataInterface() {propertyValues {Curve {defaultValue=1, keys [{tangentMode=RCTM_User, time=0, value=0}, {tangentMode=RCTM_User, time=1, value=1}]}}}
     CurveIndex=$System.Timeline001.NormalizedTime
    )
    ScaleA=true
    ScaleRGB=true
   ) {enabled=false, module=ScaleColor001}
  ]
  rendererClasses [&NiagaraSpriteRendererProperties, &NiagaraRibbonRendererProperties]
  renderers [
   {rendererClass=&NiagaraSpriteRendererProperties}
   {rendererClass=&NiagaraRibbonRendererProperties, rendererIndex=1}
  ]
 }
 SmokeTrailFE_Combined001 {
  emitterUpdateScript [
   &EmitterState()
   &SpawnParticlesFromOtherEmitter(
    Attribute Reader=DataInterface() {propertyValues {EmitterBinding {bindingMode=Other, emitterName=PlaneFE}}}
    Spawn Rate=&MakeCustomFloatFromBool(
     Bool=$User.bIsFrontEndPreview
     True Float=&Lerp_Float(A=200, B=400, Alpha=$User.ParaGlideLeanAlpha)
     False Float=0
    )
    Spawn Rate Per Particle Cap=1000
    Use Maximum Age=false
    Use Minimum Age=false
   )
  ]
  particleSpawnScript [
   &InitializeParticle(
    Lifetime=&Lerp_Float(A=0.13, B=0.25, Alpha=$System.Timeline003.NormalizedTime)
    Sprite Rotation Mode=::NewEnumerator2
    Sprite Size Mode=::NewEnumerator3
    Uniform Sprite Size=30
    UsePositionOffset=false
    Write Material Random=true
   )
   &SampleParticlesFromOtherEmitter(Mesh Orientation Sampling:ENiagara_DirectReadSamplingMode=::NewEnumerator1, Particle ID Sampling=::NewEnumerator1)
   /SetVariables(Particles.RibbonWidth=&Lerp_Float(A=30, B=60, Alpha=$User.ParaGlideLeanAlpha)) {isSetParameters=true, module=SetVariables_#11}
   /SetVariables(Particles.RibbonID=&self:RealmTrailID()) {isSetParameters=true, module=SetVariables_#12}
   /SetVariables(Particles.RibbonLinkOrder:NiagaraFloat=$Output.SampleParticlesFromOtherEmitter.SampledAge) {enabled=false, isSetParameters=true, module=SetVariables_#23}
   &ShapeLocation(
    Additional Quaternion Rotation=true
    Additional Yaw / Pitch / Roll=<0 90 90>
    Cylinder Distribution=::NewEnumerator2
    Cylinder Height=37
    Cylinder Radius=&Lerp_Float(A=1, B=4, Alpha=$User.ParaGlideLeanAlpha)
    MaskSpawnGroup=false
    Offset Mode=::NewEnumerator0
    Rotation Angle Type:ENiagara_AngleInput=::NewEnumerator0
    Rotation Mode=::NewEnumerator3
    Rotation Quaternion=$Output.SampleParticlesFromOtherEmitter.SampledMeshOrientation
    Shape Origin=$Output.SampleParticlesFromOtherEmitter.SampledPosition
    Shape Primitive=::NewEnumerator1
    U Position=&Lerp_Float(A=0, B=0, Alpha=1)
    UseOwnerScale=true
    V Position=&Lerp_Float(A=0.05, B=0.2, Alpha=$User.ParaGlideLeanAlpha)
   )
   &LinearForce(Force=&Lerp_Vector(A=<0 -300 1000>, B=<0 -1000 2000>, Alpha=$User.ParaGlideLeanAlpha)) {module=LinearForce001}
   &ApplyInitialForces()
  ]
  particleUpdateScript [
   &ParticleState()
   &SolveForcesAndVelocity(Clamp Velocity=false, Limit Acceleration=false)
   &ScaleColor(
    Scale Alpha=&FloatFromCurve(FloatCurve=DataInterface() {propertyValues {Curve {defaultValue=1, keys [<at 0 0>,<at 0.025 1>,<at 1 0>]}}})
    Scale RGB=&MakeVectorFromLinearColorRGB(Color=&ColorFromCurve(ColorCurve=DataInterface() {propertyValues {
        AlphaCurve {keys [<at 0 1 RCIM_Cubic>]}
        BlueCurve {keys [<at 0.182 0 RCIM_Cubic>,<at 0.774 0.0152 RCIM_Cubic>]}
        GreenCurve {keys [<at 0.182 0.1667 RCIM_Cubic>,<at 0.774 0.5333 RCIM_Cubic>]}
        RedCurve {keys [<at 0.182 0.0222 RCIM_Cubic>,<at 0.774 0 RCIM_Cubic>]}
       }}))
    ScaleA=true
    ScaleRGB=true
   )
   &DynamicMaterialParameters(
    Index 0 Param 1=-2.5
    Index 0 Param 4=&FloatFromCurve(FloatCurve=DataInterface() {propertyValues {Curve {keys [<at 0 0>,<at 0.01 1>,<at 0.5 1>,<at 1 0>]}}})
    Module.Index0Float0Write=true
    Module.Index0Float1Write=true
    Module.Index0Float2Write=true
    Module.Index0Float3Write=true
   ) {enabled=false}
   /SetVariables(Particles.Smoke1Dyn:Vector4f=&MakeVector4(
     X=-2.5
     Y=0
     Z=0
     W=&FloatFromCurve(FloatCurve=DataInterface() {propertyValues {Curve {keys [<at 0 0>,<at 0.01 1>,<at 0.5 1>,<at 1 0>]}}})
    )) {isSetParameters=true, module=SetVariables_#24}
   /SetVariables(Particles.Smoke2Color:LinearColor=&MakeLinearColorFromVectorAndFloat(
     "Vector (RGB)"=&MakeVectorFromLinearColorRGB(Color=$Particles.Color)
     "Float (Alpha)"=&Multiply_Float(
      A=&MakeFloatFromLinearColor(LinearColor=$Particles.Color, Channel:ENiagaraFloatFromLinearColorOptions=::NewEnumerator3)
      B=&Lerp_Float(A=1, B=0.75, Alpha=$User.ParaGlideLeanAlpha)
     )
    )) {isSetParameters=true, module=SetVariables_#25}
   &ScaleRibbonWidth(Ribbon Width Scale=&FloatFromCurve(FloatCurve=DataInterface() {propertyValues {Curve {
        defaultValue=1
        keys [
         {arriveTangent=0.5183, interpMode=RCIM_Cubic, leaveTangent=0.5183, tangentMode=RCTM_User, time=0, value=0.05}
         {interpMode=RCIM_Cubic, leaveTangent=0.8555, time=0.25, value=1}
        ]
       }}}))
   &ScaleColor(
    Scale Alpha=&FloatFromCurve(
     FloatCurve=DataInterface() {propertyValues {Curve {defaultValue=1, keys [{tangentMode=RCTM_User, time=0, value=0}, {tangentMode=RCTM_User, time=1, value=1}]}}}
     CurveIndex=$System.Timeline001.NormalizedTime
    )
    Scale RGB=&MakeVectorFromLinearColorRGB(Color=&FNTimeOfDayColor(Night Color=<rgba 0.5713 0.5615 0.7 1>, Morning Color=<rgba 1 1 1 1>, Day Color=<rgba 1 1 1 1>, Evening Color=<rgba 1 1 1 1>))
    ScaleA=true
    ScaleRGB=true
   ) {enabled=false, module=ScaleColor001}
  ]
  rendererClasses [&NiagaraSpriteRendererProperties, &NiagaraRibbonRendererProperties]
  renderers [
   {rendererClass=&NiagaraSpriteRendererProperties}
   {rendererClass=&NiagaraRibbonRendererProperties, rendererIndex=1}
   {rendererClass=&NiagaraRibbonRendererProperties, rendererIndex=2}
  ]
 }
}
summary {
 systemName=NS_Contrail_CanineCronut
 emitters [
  {emitterName=Plane, rendererClasses [&NiagaraSpriteRendererProperties, &NiagaraMeshRendererProperties]}
  {
   bEnabled=false
   emitterName=SmokeTrail
   rendererClasses [&NiagaraSpriteRendererProperties, &NiagaraRibbonRendererProperties]
  }
  {
   emitterName=LeftWingTrail
   rendererClasses [&NiagaraRibbonRendererProperties, &NiagaraSpriteRendererProperties]
  }
  {
   emitterName=RightWingTrail
   rendererClasses [&NiagaraRibbonRendererProperties, &NiagaraSpriteRendererProperties]
  }
  {
   bEnabled=false
   emitterName=SmokeTrail2
   rendererClasses [&NiagaraSpriteRendererProperties, &NiagaraRibbonRendererProperties]
  }
  {bEnabled=false, emitterName=SmokeBurst, rendererClasses [&NiagaraSpriteRendererProperties]}
  {emitterName=FN_SpeedLines001, rendererClasses [&NiagaraSpriteRendererProperties]}
  {emitterName=PlaneFE, rendererClasses [&NiagaraSpriteRendererProperties, &NiagaraMeshRendererProperties]}
  {
   bEnabled=false
   emitterName=SmokeTrail2FE
   rendererClasses [&NiagaraSpriteRendererProperties, &NiagaraRibbonRendererProperties]
  }
  {
   emitterName=LeftWingTrailFE
   rendererClasses [&NiagaraRibbonRendererProperties, &NiagaraSpriteRendererProperties]
  }
  {
   emitterName=RightWingTrailFE
   rendererClasses [&NiagaraRibbonRendererProperties, &NiagaraSpriteRendererProperties]
  }
  {
   emitterName=SmokeTrailCombined
   rendererClasses [&NiagaraSpriteRendererProperties, &NiagaraRibbonRendererProperties]
  }
  {
   emitterName=SmokeTrailFE_Combined001
   rendererClasses [&NiagaraSpriteRendererProperties, &NiagaraRibbonRendererProperties]
  }
  {
   bEnabled=false
   emitterName=SmokeTrailFE001
   rendererClasses [&NiagaraSpriteRendererProperties, &NiagaraRibbonRendererProperties]
  }
 ]
 userVariables [
  User.bIsFrontEnd {type=NiagaraBool}
  User.bIsFrontEndPreview {type=NiagaraBool}
  User.ParaGlideLeanAlpha {type=NiagaraFloat, defaultValue=0, description=""}
  User.PlayerSpeed {type=NiagaraFloat, defaultValue=0, description=""}
  User.RibbonIDOffset {type=NiagaraInt32}
 ]
}
