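# Particle material for the CanineCronut skydiving contrail, built for Niagara meshes and sprites with Default Lit shading.
# Time-driven RotateAboutAxis spins vertices around pivots rebuilt from UV1 and UV2, masked by red vertex color, and rotates the normal map to match.
# Panned textures drive particle-tinted base color, normal and specular maps.
# Emissive has optional panning and sphere masks blended by scalar lerps, while particle alpha drives noise-based opacity erosion.
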Material /BRCosmetics/Athena/Effects/Particles/Skydiving/NS_Contrail_CanineCronut/Materials/M_Contrail_CanineCronut_WPO
graph {
 properties {
  bUsedWithNiagaraMeshParticles=true
  bUsedWithNiagaraSprites=true
  pixelDepthOffsetMode=PDOM_Legacy
  previewMesh=/BRCosmetics/Athena/Effects/Particles/Skydiving/NS_Contrail_BrokenGravel/Meshes/SM_FX_BrokenGravel_Drone
  usedShadingModels=Default Lit
 }
 nodes [
  n1 = Panner(Speed=AppendVector(A=ScalarParameter() {group="0) UVs", parameterName=U Speed}, B=ScalarParameter() {group="0) UVs", parameterName=V Speed}))
  n2 = TextureSampleParameter2D(UVs=@n1) {
   group="8) Normal"
   parameterName=Normal
   samplerType=SAMPLERTYPE_Normal
   texture=/BRCosmetics/Athena/Effects/Particles/Skydiving/NS_Contrail_CanineCronut/Textures/T_CanineCronut_Contrail_N
  }
  n3 = Multiply(A=@n2.RGB, B=VectorParameter() {group="8) Normal", parameterName=NormalMultiply, defaultValue=<rgba 1 1 1 0>}.RGB)
  n4 = Transform(None=@n3)
  n5 = Transform(None=VectorParameter() {parameterName=RotationAxis, defaultValue=<rgba 0 1 0 1>}.RGB) {transformSourceType=TRANSFORMSOURCE_Local}
  n6 = Multiply(A=Time(), B=ScalarParameter() {defaultValue=-3, parameterName=RotationSpeed}) {constB=-3}
  n7 = Constant()
  n8 = Add(A=@n4, B=VertexInterpolator(VS=RotateAboutAxis(NormalizedRotationAxis=@n5, RotationAngle=@n6, PivotPoint=@n7, Position=@n4)).PS)
  n9 = VertexColor()
  LinearInterpolate(
   A=@n3
   B=Transform(None=Normalize(VectorInput=@n8)) {transformSourceType=TRANSFORMSOURCE_World, transformType=TRANSFORM_Tangent}
   Alpha=@n9.R
  )
  LinearInterpolate()
  n10 = VectorParameter() {group="4.2) Emissive Sphere Mask", parameterName=Location Offset, sortPriority=5, defaultValue=<rgba 0 0 0 1>}
  n11 = SphereMask(
   A=WorldPosition().XYZ
   B=Add(A=ActorPositionWS(), B=@n10.RGB)
   Radius=ScalarParameter() {defaultValue=50, group="4.2) Emissive Sphere Mask", parameterName=Sphere Mask Radius, sortPriority=3}
   Hardness=ScalarParameter() {group="4.2) Emissive Sphere Mask", parameterName=Sphere Mask Hardness, sortPriority=4}
  )
  n12 = ScalarParameter() {group="4.2) Emissive Sphere Mask", parameterName=Power, sortPriority=5}
  n13 = LinearInterpolate(
   A=ScalarParameter() {group="4.2) Emissive Sphere Mask", parameterName=Sphere Mask Low Value, sortPriority=1}
   B=ScalarParameter() {defaultValue=1, group="4.2) Emissive Sphere Mask", parameterName=Sphere Mask High Value, sortPriority=2}
   Alpha=Clamp(None=Power(Base=@n11, Exp=@n12) {constExponent=1})
  )
  n14 = ScalarParameter() {defaultValue=1, group="4) Emissive", parameterName=Emissive Mask Texture Scale U}
  n15 = ScalarParameter() {defaultValue=1, group="4) Emissive", parameterName=Emissive Mask Texture Scale V}
  n16 = TextureSampleParameter2D(UVs=Multiply(A=TextureCoordinate(), B=AppendVector(A=@n14, B=@n15))) {group="4) Emissive", parameterName=Emissive Texture, sortPriority=1, texture=&T_BlacksRGB}
  n17 = ChannelMaskParameter(None=@n16.RGBA) {group="4) Emissive", maskChannel=Green, parameterName=Emissive Mask, defaultValue=<rgba 0 1 0 0>}
  n18 = TextureCoordinate()
  n19 = AppendVector(
   A=ScalarParameter() {defaultValue=1, group="4.1) Emissive Masking", parameterName=Emissive Mask Multiplier Scale U}
   B=ScalarParameter() {defaultValue=1, group="4.1) Emissive Masking", parameterName=Emissive Mask Multiplier Scale V}
  )
  n20 = ScalarParameter() {defaultValue=1, group="4.1) Emissive Masking", parameterName=Emissive Mask Multiplier Pan Speed U}
  n21 = ScalarParameter() {defaultValue=1, group="4.1) Emissive Masking", parameterName=Emissive Mask Multiplier Pan Speed V}
  n22 = Multiply(A=Time(), B=AppendVector(A=@n20, B=@n21))
  n23 = ScalarParameter() {defaultValue=4, group="4.1) Emissive Masking", parameterName=Emissive Mask Offset Steps}
  n24 = AppendVector(
   A=MaterialFunctionCall!(TimeSpeed=@n20, Jump Amount Per Step=@n23, Number of Steps Per Unit of Time=@n23, Time Offset) {materialFunction=/Engine/Functions/Engine_MaterialFunctions02/SteppingPannerTime}.Result
   B=MaterialFunctionCall(TimeSpeed=@n21, Jump Amount Per Step=@n23, Number of Steps Per Unit of Time=@n23) {materialFunction=&SteppingPannerTime}.Result
  )
  n25 = ScalarParameter() {group="4.1) Emissive Masking", parameterName=Use Stepped Panner For Emissive Multiplier Mask}
  n26 = TextureSampleParameter2D(UVs=Add(A=Multiply(A=@n18, B=@n19), B=LinearInterpolate(A=@n22, B=@n24, Alpha=@n25))) {
   group="4.1) Emissive Masking"
   parameterName=Emissive Multiplier Texture
   texture=/Game/ContentCreationTools/ObjectPlacement/Textures/T_BlankWhite_01
  }
  n27 = ScalarParameter() {defaultValue=2, group="4.1) Emissive Masking", parameterName=Emissive Mask Multiplier Power}
  n28 = Add(
   A=Add(A=TextureCoordinate(), B=Constant2Vector() {g=0.25, r=0.25})
   B=AppendVector(
    A=ScalarParameter() {group="4.1) Emissive Masking", parameterName=Emissive Mask Multiplier Offset U}
    B=ScalarParameter() {group="4.1) Emissive Masking", parameterName=Emissive Mask Multiplier Offset V}
   )
  )
  n29 = Add(
   A=ComponentMask(None=Sine(None=@n28)) {r=true}
   B=LinearInterpolate(
    A=Constant() {r=-1}
    B=Constant() {r=1}
    Alpha=ScalarParameter() {group="4.1) Emissive Masking", parameterName=Emissive Mask Multiplier Thickness U}
   )
  )
  n30 = OneMinus(None=Clamp(None=Ceil(None=@n29)))
  n31 = Add(
   A=ComponentMask(None=Sine(None=@n28)) {g=true}
   B=LinearInterpolate(
    A=Constant() {r=-1}
    B=Constant() {r=1}
    Alpha=ScalarParameter() {group="4.1) Emissive Masking", parameterName=Emissive Mask Multiplier Thickness V}
   )
  )
  n32 = OneMinus(None=Clamp(None=Ceil(None=@n31)))
  n33 = Multiply(A=@n17.R, B=Multiply(A=Power(Base=@n26.RGB, Exp=@n27), B=Multiply(A=@n30, B=@n32)))
  n34 = LinearInterpolate(
   A=@n17.R
   B=@n33
   Alpha=ScalarParameter() {desc=This is a lerp to avoid a static switch., group="4.1) Emissive Masking", parameterName=Use Emissive Mask}
  )
  n35 = Multiply(A=@n34, B=@n13)
  n36 = ScalarParameter() {group="4.2) Emissive Sphere Mask", parameterName="Use Sphere Mask (World)", sortPriority=-1}
  n37 = VectorParameter() {group="4) Emissive", parameterName=Emissive Multiplier, defaultValue=<rgba 1 1 1 0>}
  n38 = ParticleColor()
  n39 = ScalarParameter() {group="4) Emissive", parameterName=UseParticleColorForEmissive}
  MaterialFunctionCall(Emissive=Multiply(A=LinearInterpolate(A=@n34, B=@n35, Alpha=@n36), B=LinearInterpolate(A=@n37.RGB, B=@n38.RGB, Alpha=@n39))) {materialFunction=&MF_FX_EmissiveAdjustments}
  n40 = ParticleColor()
  n41 = TextureSampleParameter2D(UVs=@n1) {
   group="1) Base Color"
   parameterName=Diffuse
   texture=/BRCosmetics/Athena/Effects/Particles/Skydiving/NS_Contrail_CanineCronut/Textures/T_CanineCronut_Contrail_D
  }
  n42 = TextureSampleParameter2D(UVs=@n1) {group="1) Base Color", parameterName=BaseColorMask, samplerType=SAMPLERTYPE_Masks, texture=/BRCosmetics/Weapons/Ammo/Textures/T_AmmoSet02_S}
  n43 = ChannelMaskParameter(None=@n42.RGBA) {group="1) Base Color", maskChannel=Green, parameterName=BaseColorMaskChannel, defaultValue=<rgba 0 1 0 0>}
  n44 = LinearInterpolate(
   A=Multiply(A=@n40.RGB, B=@n41.RGB)
   B=LinearInterpolate(A=@n40.RGB, B=@n41.RGB, Alpha=@n43.R)
   Alpha=ScalarParameter() {group="1) Base Color", parameterName=UseBaseColorMask}
  )
  Multiply(A=@n40.RGB, B=@n44)
  n45 = VectorParameter() {group="1.1) Color Multiplier And Saturation", parameterName=Base Color Multiplier, defaultValue=<rgba 0.5 0.5 0.5 1>}
  n46 = Desaturation(None=Multiply(A=@n44, B=@n45.RGB), Fraction=ScalarParameter() {group="1.1) Color Multiplier And Saturation", parameterName=Desaturation})
  n47 = VectorParameter() {group="1.1) Color Multiplier And Saturation", parameterName=FoliageSubsurfaceColor, defaultValue=<rgba 1 1 1 0>}
  Multiply(
   A=LinearInterpolate(A=@n46, B=Multiply(A=@n46, B=@n13), Alpha=ScalarParameter() {group="4.2) Emissive Sphere Mask", parameterName=Base Color Influence})
   B=@n47.RGB
  )
  n48 = TextureCoordinate() {coordinateIndex=1}
  n49 = VectorParameter() {parameterName=RotationPivotMultiplier, defaultValue=<rgba 10 10 4.5 1>}
  n50 = ComponentMask(None=@n49.RGB) {g=true, r=true}
  n51 = VectorParameter() {parameterName=RotationPivotOffset, defaultValue=<rgba 0 0 0 1>}
  n52 = ComponentMask(None=TextureCoordinate() {coordinateIndex=2}) {g=true}
  n53 = AppendVector(
   A=Add(A=Multiply(A=@n48, B=@n50) {constB=10}, B=ComponentMask(None=@n51.RGB) {g=true, r=true})
   B=Add(A=Multiply(A=@n52, B=@n49.B) {constB=4}, B=@n51.B)
  )
  Multiply(
   A=RotateAboutAxis(
    NormalizedRotationAxis=@n5
    RotationAngle=@n6
    PivotPoint=TransformPosition(None=@n53) {transformType=TRANSFORMPOSSOURCE_World}
    Position=WorldPosition().XYZ
   )
   B=@n9.R
  )
  n54 = TextureSampleParameter2D() {group="3) Opacity", parameterName=Opacity, sortPriority=1, texture=/Game/Global/Textures/Default/Blanks/T_WhiteUserInterface}
  n55 = ScalarParameter() {defaultValue=1, group="3.1) Erosion", parameterName=Erosion UV Scale U}
  n56 = ScalarParameter() {defaultValue=1, group="3.1) Erosion", parameterName=Erosion UV Scale V}
  n57 = TextureSampleParameter2D(UVs=Multiply(A=TextureCoordinate(), B=AppendVector(A=@n55, B=@n56))) {
   group="3.1) Erosion"
   parameterName=ErosionTextureMask
   samplerType=SAMPLERTYPE_LinearColor
   texture=/Game/Global/Textures/Default/Noise/Noise32x32
  }
  n58 = ParticleColor()
  n59 = Floor(None=Subtract(A=Add(A=@n57.R), B=LinearInterpolate(Alpha=@n58.A) {constA=1.2, constB=0}))
  Multiply(
   A=ChannelMaskParameter(None=@n54.RGBA) {group="3) Opacity", maskChannel=Green, parameterName=Opacity Mask, defaultValue=<rgba 0 1 0 0>}.R
   B=LinearInterpolate(B=Clamp(None=@n59), Alpha=ScalarParameter() {defaultValue=1, parameterName=UseParticleAlphaErosion}) {constA=1}
  )
  n60 = TextureSampleParameter2D(UVs=@n1) {
   group="2) Specular"
   parameterName=Specular
   samplerType=SAMPLERTYPE_Masks
   sortPriority=1
   texture=/BRCosmetics/Athena/Effects/Particles/Skydiving/NS_Contrail_CanineCronut/Textures/T_CanineCronut_Contrail_S
  }
  OneMinus(None=@n60.B)
  StaticSwitchParameter(True=@n33, False=@n17.R) {group=Features, parameterName=Use Emissive Multiplier Masking}
 ]
}
